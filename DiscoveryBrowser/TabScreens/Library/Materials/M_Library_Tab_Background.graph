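# UI translucent material for a Library tab button background in the Discovery Browser, with normal and selected states.
# The Slate post buffer feeds a CoreUI backplate function, and a slight blend of that result over the raw buffer supplies highlight looks.
# Base button-state and selectable-state lerp functions are chained to pick the look.
# The resulting opacity is multiplied by a rounded-corner mask with parameterized radius.
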Material /DiscoveryBrowser/TabScreens/Library/Materials/M_Library_Tab_Background
graph {
 properties {
  bUsedWithStaticMesh=false
  blendMode=BLEND_Translucent
  materialDomain=MD_UI
  pixelDepthOffsetMode=PDOM_Legacy
  usedShadingModels=Default Lit
 }
 nodes [
  n1 = MaterialFunctionCall!("UVs (Optional)") {materialFunction=/Engine/Functions/UserInterface/GetSlatePost0}
  n2 = MaterialFunctionCall!(Post Buffer Blur=@n1.LinearRGB) {materialFunction=/FortUILibrary/MaterialFunctions/CoreUI/MF_UI_CoreUI_BackplateA}
  n3 = LinearInterpolate(A=@n1.LinearRGB, B=@n2.Result) {constAlpha=0.1}
  n4 = Constant() {r=1}
  n5 = MaterialFunctionCall!(
   Idle
   Idle Opacity
   Hovered=@n3
   Hovered Opacity=@n4
   Focused=@n3
   Focused Opacity=@n4
   Pressed=@n3
   Pressed Opacity=@n4
   Disabled
   Disabled Opacity
  ) {materialFunction=/FortUILibrary/Systems/ModularButton/MaterialFunctions/MF_UIKit_ButtonStatesBase_Lerp}
  n6 = Constant() {r=1}
  n7 = VectorParameter() {parameterName=Corner Radius, defaultValue=<rgba 16 16 16 16>}
  Multiply(
   A=MaterialFunctionCall!(
    Base States=@n5.Color
    Base States Opacity=@n5.Color
    Selected_Idle=@n2.Result
    Selected_Idle Opacity=@n6
    Selected_Hovered=@n3
    Selected_Hovered Opacity=@n6
    Selected_Focused=@n2.Result
    Selected_Focused Opacity=@n6
    Selected_Pressed=@n2.Result
    Selected_Pressed Opacity=@n6
    Selected_Disabled
    Selected_Disabled Opacity
   ) {materialFunction=/FortUILibrary/Systems/ModularButton/MaterialFunctions/MF_UIKit_ButtonStatesSelectable_Lerp}.Opacity
   B=MaterialFunctionCall(
    Corner Radius=AppendVector(A=@n7.RGB, B=@n7.RGB)
    Radius in Pixels=ScalarParameter() {defaultValue=1, parameterName=Radius in Pixels, sliderMax=1}
   ) {materialFunction=&MF_UI_CoreUI_RoundedMask}.Shape Mask
  )
 ]
}
