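# UI material for the Library tab storage bar in the Discovery Browser, showing Fortnite content, other content and free disk space.
# Three scalar parameters supply Fortnite, total content and disk sizes, with content values remapped to 0-1 to drive the bar.
# A hard horizontal mask splits yellow Fortnite from blue-gradient other content over dark grey free space.
# MF_UI_Box shapes the bar, using left-center anchored UV scaling and per-corner roundness for the content fill.
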Material /DiscoveryBrowser/TabScreens/Library/Materials/M_UI_Library_StorageBar
graph {
 properties {
  bAutomaticallySetUsageInEditor=false
  bUsedWithStaticMesh=false
  blendMode=BLEND_Translucent
  materialDomain=MD_UI
  usedShadingModels=Default Lit
 }
 nodes [
  n1 = MaterialFunctionCall() {materialFunction=&GetUserInterfaceUV}
  n2 = NamedRerouteDeclaration(None=LinearInterpolate(A=Constant2Vector() {g=1, r=1}, B=Constant2Vector() {g=-1, r=-1}, Alpha=Constant2Vector() {desc=Left Center Scaling Anchor, g=0.5}) {constA=1, constB=-1}) {name=Anchor Range, nodeColor=<rgba 0 0 0 1>}
  n3 = LinearInterpolate(A=Constant2Vector() {g=1, r=1}, B=Constant2Vector(), Alpha=NamedRerouteUsage())
  n4 = NamedRerouteDeclaration(None=Add(A=NamedRerouteDeclaration(None=@n1.Normalized UV) {name=Normalized UV, nodeColor=<rgba 0 0 0 1>}, B=Multiply(A=@n2, B=@n3))) {name=Bar UV, nodeColor=<rgba 0 0 0 1>}
  n5 = Constant() {r=1}
  n6 = Constant()
  n7 = NamedRerouteUsage()
  n8 = MaterialFunctionCall(R=@n6, G=@n6, B=@n7, A=@n7) {materialFunction=&AppendMany}
  n9 = StaticBool() {value=true}
  LinearInterpolate(
   A=Constant3Vector() {desc=Free Visual, constant=<rgba 0.0666 0.0666 0.0666 1>}
   B=NamedRerouteUsage()
   Alpha=NamedRerouteDeclaration(None=MaterialFunctionCall(UVs=@n4, Size=@n5, CornerRoundness=@n8.RGBA, KeepAspectRatio=@n9) {materialFunction=&MF_UI_Box}.Fill) {name=Content Bar Mask, nodeColor=<rgba 0 0 0 1>}
  )
  MaterialFunctionCall(
   UVs=NamedRerouteUsage()
   Size=Constant() {r=1}
   CornerRoundness=NamedRerouteDeclaration(None=Constant() {r=1}) {name=Corner Radius, nodeColor=<rgba 0 0 0 1>}
   KeepAspectRatio=StaticBool() {value=true}
  ) {materialFunction=&MF_UI_Box}
  NamedRerouteDeclaration(None=ScalarParameter() {defaultValue=1024, parameterName=Total Disk Raw}) {name=Total Disk Raw, nodeColor=<rgba 0 0 0 1>}
  n10 = Constant3Vector() {constant=<rgba 0 0.0976 0.521 1>}
  n11 = Constant3Vector() {constant=<rgba 0.0482 0.3467 0.9473 1>}
  n12 = NamedRerouteUsage()
  n13 = NamedRerouteUsage()
  n14 = Saturate(None=MaterialFunctionCall(Input=NamedRerouteUsage(), Input Low=@n12, Input High=Max(A=@n12, B=@n13)) {materialFunction=&RemapValue_0-1}.Result)
  NamedRerouteDeclaration(None=LinearInterpolate(
    A=Constant3Vector() {desc=Fortnite Visual, constant=<rgba 0.9301 1 0.0103 1>}
    B=LinearInterpolate(A=@n10, B=@n11, Alpha=@n14) {desc=Non Fortnite Visual}
    Alpha=NamedRerouteUsage()
   )) {name=Content Visual, nodeColor=<rgba 0 0 0 1>}
  n15 = NamedRerouteUsage()
  n16 = Constant()
  n17 = NamedRerouteUsage()
  NamedRerouteDeclaration(None=Saturate(None=MaterialFunctionCall(Input=@n15, Input Low=@n16, Input High=@n17) {materialFunction=&RemapValue_0-1}.Result)) {name=Fortnite Content Remapped, nodeColor=<rgba 0 0 0 1>}
  n18 = NamedRerouteUsage()
  n19 = Subtract(
   A=NamedRerouteDeclaration(None=ComponentMask(None=@n18) {r=true}) {name=Horizontal Gradient, nodeColor=<rgba 0 0 0 1>}
   B=NamedRerouteUsage()
  )
  NamedRerouteDeclaration(None=Ceil(None=Saturate(None=@n19))) {name=Content Mask, nodeColor=<rgba 0 0 0 1>}
  n20 = NamedRerouteUsage()
  n21 = Constant()
  n22 = NamedRerouteUsage()
  NamedRerouteDeclaration(None=Saturate(None=MaterialFunctionCall(Input=@n20, Input Low=@n21, Input High=@n22) {materialFunction=&RemapValue_0-1}.Result)) {name=Total Content Remapped, nodeColor=<rgba 0 0 0 1>}
  NamedRerouteDeclaration(None=ScalarParameter() {defaultValue=100, parameterName=Fortnite Content Raw}) {name=Fortnite Content Raw, nodeColor=<rgba 0 0 0 1>}
  NamedRerouteDeclaration(None=ScalarParameter() {defaultValue=512, parameterName=Total Content Raw}) {name=Total Content Raw, nodeColor=<rgba 0 0 0 1>}
 ]
}
