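# Translucent UI material for the background of a storage item in the Discovery Browser's Library tab.
# Slate post-buffer blur feeds CoreUI backplate functions, declared as named reroutes alongside Quiet, Regular and Loud CTA colors.
# Display Regular and Display Loud blend the CTA styles, which feed modular-button state lerps with a separate disabled input.
# An Is Selectable static switch adds selected states, and a rounded mask function takes corner radius parameters.
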Material /DiscoveryBrowser/TabScreens/Library/Materials/M_UI_Library_StorageItem_Bg
graph {
 properties {
  bAutomaticallySetUsageInEditor=false
  bUsedWithStaticMesh=false
  blendMode=BLEND_Translucent
  materialDomain=MD_UI
  usedShadingModels=Default Lit
 }
 nodes [
  MaterialFunctionCall(
   Corner Radius=ScalarParameter() {defaultValue=10, parameterName=Corner Radius}
   Radius in Pixels=ScalarParameter() {defaultValue=1, parameterName=Radius in Pixels}
  ) {materialFunction=&MF_UI_CoreUI_RoundedMask}
  NamedRerouteDeclaration(None=MaterialFunctionCall!("UVs (Optional)") {materialFunction=/Engine/Functions/UserInterface/GetSlatePost0}.LinearRGB) {name=Blur RGB, nodeColor=<rgba 0 0 0 1>}
  NamedRerouteDeclaration(None=MaterialFunctionCall!(Post Buffer Blur=NamedRerouteUsage()) {materialFunction=/FortUILibrary/MaterialFunctions/CoreUI/MF_UI_CoreUI_BackplateA}.Result) {name=Backplate A, nodeColor=<rgba 0 0 0 1>}
  NamedRerouteDeclaration(None=MaterialFunctionCall!(Post Buffer Blur=NamedRerouteUsage()) {materialFunction=/FortUILibrary/MaterialFunctions/CoreUI/MF_UI_CoreUI_BackplateB}.Result) {name=Backplate B, nodeColor=<rgba 0 0 0 1>}
  NamedRerouteDeclaration(None=NamedRerouteUsage()) {name=Interaction Color, nodeColor=<rgba 0 0 0 1>}
  NamedRerouteDeclaration(None=MaterialFunctionCall!(Post Buffer Blur=NamedRerouteUsage()) {materialFunction=/FortUILibrary/MaterialFunctions/CoreUI/MF_UI_CoreUI_BackplateC}.Result) {name=Backplate C, nodeColor=<rgba 0 0 0 1>}
  NamedRerouteDeclaration(None=MaterialFunctionCall!(Post Buffer Blur=NamedRerouteUsage()) {materialFunction=/FortUILibrary/MaterialFunctions/CoreUI/MF_UI_CoreUI_BackplateDisabled}.Result) {name=Backplate Disabled, nodeColor=<rgba 0 0 0 1>}
  NamedRerouteUsage()
  n1 = NamedRerouteUsage()
  n2 = NamedRerouteUsage()
  n3 = LinearInterpolate(
   A=LinearInterpolate(
    A=NamedRerouteDeclaration(None=@n1) {name=Quiet CTA, nodeColor=<rgba 0 0 0 1>}
    B=NamedRerouteDeclaration(None=@n2) {name=Regular CTA, nodeColor=<rgba 0 0 0 1>}
    Alpha=ScalarParameter() {defaultValue=1, group=Style, parameterName=Display Regular, sliderMax=1}
   )
   B=NamedRerouteDeclaration(None=Constant3Vector() {constant=<rgba 0.9301 1 0.0103 1>}) {name=Loud CTA, nodeColor=<rgba 0 0 0 1>}
   Alpha=ScalarParameter() {defaultValue=1, group=Style, parameterName=Display Loud, sliderMax=1}
  )
  n4 = MaterialFunctionCall!(
   Idle=@n3
   Idle Opacity
   Hovered=@n3
   Hovered Opacity
   Focused=@n3
   Focused Opacity
   Pressed=@n3
   Pressed Opacity
   Disabled=NamedRerouteUsage()
   Disabled Opacity
  ) {materialFunction=/FortUILibrary/Systems/ModularButton/MaterialFunctions/MF_UIKit_ButtonStatesBase_Lerp}
  StaticSwitchParameter(
   True=MaterialFunctionCall!(
    Base States=@n4.Color
    Base States Opacity
    Selected_Idle
    Selected_Idle Opacity
    Selected_Hovered
    Selected_Hovered Opacity
    Selected_Focused
    Selected_Focused Opacity
    Selected_Pressed
    Selected_Pressed Opacity
    Selected_Disabled
    Selected_Disabled Opacity
   ) {materialFunction=/FortUILibrary/Systems/ModularButton/MaterialFunctions/MF_UIKit_ButtonStatesSelectable_Lerp}.Color
   False=@n4.Color
  ) {parameterName=Is Selectable}
 ]
}
