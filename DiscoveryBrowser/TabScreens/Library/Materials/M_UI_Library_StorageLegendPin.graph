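# Translucent UI material for a storage legend pin in the Library tab, built from a two-color gradient and a circle shape.
# Two color parameters, Color 1 and Color 2, blend horizontally along the red channel of the normalized UI UV.
# The MF_UI_Circle function generates the full-size circle procedurally.
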Material /DiscoveryBrowser/TabScreens/Library/Materials/M_UI_Library_StorageLegendPin
graph {
 properties {
  bAutomaticallySetUsageInEditor=false
  bUsedWithStaticMesh=false
  blendMode=BLEND_Translucent
  materialDomain=MD_UI
  usedShadingModels=Default Lit
 }
 nodes [
  LinearInterpolate(
   A=VectorParameter() {parameterName=Color 1, defaultValue=<rgba 0.9208 0 1 1>}.RGB
   B=VectorParameter() {parameterName=Color 2, defaultValue=<rgba 0 1 0.0637 1>}.RGB
   Alpha=ComponentMask(None=MaterialFunctionCall() {materialFunction=&GetUserInterfaceUV}.Normalized UV) {r=true}
  )
  MaterialFunctionCall(Size=Constant() {r=1}) {materialFunction=&MF_UI_Circle}
 ]
}
