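# Translucent UI material for library storage thumbnails that shows download progress and a removing state.
# The Thumbnail texture is sampled with clamped, size-aware UVs from MF_UI_UVScaling, alongside a rounded-corner mask function.
# A Ceil of horizontal UV minus DownloadProgress forms a hard edge, desaturating the image beyond the progress point.
# IsRemoving blends in a multiplier that darkens the result.
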Material /DiscoveryBrowser/TabScreens/Library/Materials/M_UI_Library_Storage_Thumbnail
graph {
 properties {
  bAutomaticallySetUsageInEditor=false
  bUsedWithStaticMesh=false
  blendMode=BLEND_Translucent
  materialDomain=MD_UI
  usedShadingModels=Default Lit
 }
 nodes [
  MaterialFunctionCall(Corner Radius=Constant() {r=10}, Radius in Pixels=Constant() {r=1}) {materialFunction=&MF_UI_CoreUI_RoundedMask}
  n1 = MaterialFunctionCall() {materialFunction=&GetUserInterfaceUV}
  n2 = TextureObjectParameter() {parameterName=Thumbnail}
  n3 = TextureSample(
   UVs=MaterialFunctionCall(UV=NamedRerouteDeclaration(None=@n1.Normalized UV) {name=Normalized UV, nodeColor=<rgba 0 0 0 1>}, TextureSize=TextureProperty(Tex=@n2)) {materialFunction=&MF_UI_UVScaling}.ScaleY
   Tex=@n2
  ) {samplerSource=SSM_Clamp_WorldGroupSettings}
  n4 = NamedRerouteUsage()
  n5 = ScalarParameter() {defaultValue=0.5, parameterName=DownloadProgress, sliderMax=1}
  n6 = Saturate(None=Subtract(A=ComponentMask(None=@n4) {r=true}, B=NamedRerouteDeclaration(None=@n5) {name=Download Progress, nodeColor=<rgba 0 0 0 1>}))
  Multiply(
   A=LinearInterpolate(A=@n3.RGB, B=Desaturation(None=@n3.RGB), Alpha=Ceil(None=@n6)) {constAlpha=0}
   B=LinearInterpolate(Alpha=ScalarParameter() {parameterName=IsRemoving, sliderMax=1}) {constA=1, constAlpha=1, constB=0.4}
  )
 ]
}
