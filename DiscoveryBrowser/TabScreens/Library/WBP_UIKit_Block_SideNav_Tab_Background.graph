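# Button background block for side-nav tabs in the Discovery Browser Library, extending WBP_UIKit_Block_BackgroundGeneric without adding widgets.
# PreConstruct extends the outline beyond its slot with negative padding and rounds both dynamic materials' corners.
# Focus and unfocus play material-driven outline animations, and selected tabs also inset the fill and tighten its corners.
# Instant selected and enabled transitions stop running animations and snap by starting playback at the animations' end time.
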WidgetBlueprint /DiscoveryBrowser/TabScreens/Library/WBP_UIKit_Block_SideNav_Tab_Background
animations [
 Anim_HoverOutline {
  binding_count=1
  length_seconds=1
  object=&self:Anim_HoverOutline
  sequence {
   bindings [Outline {
     id=#1
     possessed_class=FortMobileImage
     tracks [{
       type=MovieSceneWidgetMaterialTrack
       sections [{
         type=MovieSceneParameterSection
         channels [
          Anim_Hover_Fill {type=MovieSceneScriptingFloatChannel, keys [[0,0],[60,1]]}
          Anim_Hover_Energy {type=MovieSceneScriptingFloatChannel, keys [[0,0],[45,1]]}
          Anim_Hover_Outline {type=MovieSceneScriptingFloatChannel, keys [[0,0],[15,1]]}
          Anim_OutlineThickness {type=MovieSceneScriptingFloatChannel, keys [[0,0],[15,1]]}
          FInalAlpha {type=MovieSceneScriptingFloatChannel, keys [[0,0],[3,1]]}
         ]
        }]
      }]
    }]
   display_rate {numerator=60}
  }
 }
 Anim_UnhoverOutline {
  binding_count=1
  length_seconds=0.5
  object=&self:Anim_UnhoverOutline
  sequence {
   bindings [Outline {
     id=#1
     possessed_class=FortMobileImage
     tracks [{
       type=MovieSceneWidgetMaterialTrack
       sections [{
         type=MovieSceneParameterSection
         channels [
          Anim_Hover_Fill {type=MovieSceneScriptingFloatChannel, const [1,2]}
          Anim_Hover_Energy {type=MovieSceneScriptingFloatChannel, const [1,3]}
          Anim_Hover_Outline {type=MovieSceneScriptingFloatChannel, const [1,2]}
          Anim_OutlineThickness {type=MovieSceneScriptingFloatChannel, keys [[0,1],[12,0]]}
          FInalAlpha {type=MovieSceneScriptingFloatChannel, keys [[0,1],[6,1],[30,0]]}
         ]
        }]
      }]
    }]
   display_rate {numerator=60}
  }
 }
 Anim_SelectedFocused {
  binding_count=2
  length_seconds=0.2
  object=&self:Anim_SelectedFocused
  sequence {bindings [
    GridSlot_0 {
     id=#2
     parent_id=#3
     possessed_class=GridSlot
     tracks [{
       type=MovieSceneMarginTrack
       display_name=Padding
       sections [{
         type=MovieSceneMarginSection
         end_frame=4
         start_frame=0
         channels [
          Left {type=MovieSceneScriptingFloatChannel, keys [[0,0],[4,8]]}
          Top {type=MovieSceneScriptingFloatChannel, keys [[0,0],[4,8]]}
          Right {type=MovieSceneScriptingFloatChannel, keys [[0,0],[4,8]]}
          Bottom {type=MovieSceneScriptingFloatChannel, keys [[0,0],[4,8]]}
         ]
        }]
      }]
    }
    Fill {
     id=#3
     possessed_class=FortMobileImage
     tracks [{
       type=MovieSceneWidgetMaterialTrack
       sections [{
         type=MovieSceneParameterSection
         end_frame=4
         start_frame=0
         channels [
          Corner RadiusR {type=MovieSceneScriptingFloatChannel, keys [[0,16],[4,8]]}
          Corner RadiusG {type=MovieSceneScriptingFloatChannel, keys [[0,16],[4,8]]}
          Corner RadiusB {type=MovieSceneScriptingFloatChannel, keys [[0,16],[4,8]]}
          Corner RadiusA {type=MovieSceneScriptingFloatChannel, keys [[0,16],[4,8]]}
         ]
        }]
      }]
    }
   ]}
 }
]
graphs [EventGraph {nodes [
   Event OnInstantTransitionToSelected() {is=title K2Node_Event}
   > Parent: OnInstantTransitionToSelected() {is=title K2Node_CallParentFunction}
   > StopAllAnimations(self:User Widget Object Reference)
   n1 = Get Anim_HoverOutline() -> (Anim_HoverOutline:Widget Animation Object Reference)
   > PlayAnimation(InAnimation=@n1, StartAtTime=GetEndTime(self=@n1))
   n2 = Get Anim_SelectedFocused() -> (Anim_SelectedFocused:Widget Animation Object Reference)
   > PlayAnimation(InAnimation=@n2, StartAtTime=GetEndTime(self=@n2))
   n3 = Event PreConstruct()
   > Parent: PreConstruct(IsDesignTime=@n3.IsDesignTime)
   > Set!(
    execute
    Padding_Left:float=-8
    Padding_Top:float=-8
    Padding_Right:float=-8
    Padding_Bottom:float=-8
    self:Grid Slot Object Reference=SlotAsGridSlot(Widget=Get Outline() -> (Outline:Fort Mobile Image Object Reference))
   ) -> (then, Output_Get:Margin Structure)
   n4 = Get OutlineDynamicMaterial() -> (OutlineDynamicMaterial:Material Instance Dynamic Object Reference)
   n5 = Get BackgroundDynamicMaterial() -> (BackgroundDynamicMaterial:Material Instance Dynamic Object Reference)
   > SetScalarParameterValue(self=[@n4, @n5], ParameterName=Radius in Pixels, Value=1)
   > SetVectorParameterValue!(execute, self=[@n4, @n5], ParameterName=Corner Radius, Value_R:float=16, Value_G:float=16, Value_B:float=16, Value_A:float=16) -> (then)
   Event OnTransitionFocusedRequested() {is=title K2Node_Event}
   > Parent: OnTransitionFocusedRequested() {is=title K2Node_CallParentFunction}
   > StopAnimation(InAnimation=Get Anim_UnhoverOutline() -> (Anim_UnhoverOutline:Widget Animation Object Reference))
   > PlayAnimationForward(InAnimation=Get Anim_HoverOutline())
   n6 = GetIsSelected!() -> (IsSelected:bool)
   > Branch(Condition=@n6)
   n7 = Get Anim_SelectedFocused()
   > PlayAnimationForward(InAnimation=@n7)
   Event OnTransitionUnfocusedRequested() {is=title K2Node_Event}
   > Parent: OnTransitionUnfocusedRequested() {is=title K2Node_CallParentFunction}
   > StopAnimation(InAnimation=Get Anim_HoverOutline())
   > PlayAnimationForward(InAnimation=Get Anim_UnhoverOutline())
   > Branch(Condition=@n6)
   > PlayAnimationReverse(InAnimation=@n7)
   Event OnTransitionSelectedRequested() {is=title K2Node_Event}
   > Parent: OnTransitionSelectedRequested() {is=title K2Node_CallParentFunction}
   Event OnTransitionDeselectedIdleRequested() {is=title K2Node_Event}
   > Parent: OnTransitionDeselectedIdleRequested() {is=title K2Node_CallParentFunction}
   Event OnInstantTransitionToEnabled() {is=title K2Node_Event}
   > Parent: OnInstantTransitionToEnabled() {is=title K2Node_CallParentFunction}
   > StopAllAnimations()
   n8 = Get Anim_UnhoverOutline()
   > PlayAnimation(InAnimation=@n8, StartAtTime=GetEndTime(self=@n8))
   > PlayAnimation(InAnimation=@n2, StartAtTime=GetEndTime(self=@n2), PlayMode=Reverse)
   Event OnTransitionDeselectedFocusedRequested() {is=title K2Node_Event}
   > Parent: OnTransitionDeselectedFocusedRequested() {is=title K2Node_CallParentFunction}
  ]}]
variables []
widgets {
 info {
  inheritedWidgetCount=0
  parentClass=/FortUILibrary/Systems/ModularButton/Blocks/WBP_UIKit_Block_BackgroundGeneric.WBP_UIKit_Block_BackgroundGeneric_C
  rootWidgetClass=null
  widgetCount=0
 }
 widgets []
}
