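# Storage bar widget for the Discover Library screen, showing Fortnite content, total content and total disk storage as a single bar.
# On pre-construct it caches a dynamic material from the ProgressBar image, which draws MI_UI_Library_StorageBar over a rounded backplate.
# Three setter functions write raw sizes into that material's scalar parameters, leaving the bar's proportions and fill to the material rather than the widget.
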WidgetBlueprint /DiscoveryBrowser/TabScreens/Library/WBP_UI_Discover_Library_StorageBar
graphs [
 EventGraph {nodes [
   Event PreConstruct()
   > n1 = GetDynamicMaterial(self=Get ProgressBar() -> (ProgressBar:Image Object Reference))
   > Set DynamicProgressMaterial(DynamicProgressMaterial:Material Instance Dynamic Object Reference=@n1) -> (Output_Get:Material Instance Dynamic Object Reference)
   Event Construct()
  ]}
 SetFortniteStorageSize {nodes [
   n1 = SetFortniteStorageSize!() -> (then, Size:float=~) {is=title K2Node_FunctionEntry}
   > SetScalarParameterValue(
    self=Get DynamicProgressMaterial() -> (DynamicProgressMaterial:Material Instance Dynamic Object Reference)
    ParameterName=Fortnite Content Raw
    Value=@n1
   )
  ]}
 SetTotalContentStorageSize {nodes [
   n1 = SetTotalContentStorageSize!() -> (then, Size:float=~) {is=title K2Node_FunctionEntry}
   > SetScalarParameterValue(self=Get DynamicProgressMaterial(), ParameterName=Total Content Raw, Value=@n1)
  ]}
 SetTotalDiskStorageSize {nodes [
   n1 = SetTotalDiskStorageSize!() -> (then, Size:float=~) {is=title K2Node_FunctionEntry}
   > SetScalarParameterValue(self=Get DynamicProgressMaterial(), ParameterName=Total Disk Raw, Value=@n1)
  ]}
]
variables [DynamicProgressMaterial {type=MaterialInstanceDynamic}]
widgets {
 info {inheritedWidgetCount=0, parentClass=&CommonUserWidget, rootWidgetClass=&GridPanel, widgetCount=4}
 widgets [
  GridPanel_23 {widgetClassPath=&GridPanel, widget=GridPanel_23, properties {columnFill [1], rowFill [1]}}
  Min_Size {
   widgetClassPath=&Spacer
   parent=GridPanel_23
   widget=Min_Size
   properties {size=<368 16>}
   slot {class=GridSlot, properties {layer=-1}}
  }
  Backplate {
   widgetClassPath=&WBP_UIKit_Backplate
   parent=GridPanel_23
   widget=Backplate
   properties {brightness=Obscure, cornerRadius=Circle}
   slot {class=GridSlot}
  }
  ProgressBar {
   widgetClassPath=&Image
   bIsVariable=true
   parent=GridPanel_23
   widget=ProgressBar
   properties {
    visibility=HitTestInvisible
    brush {
     bIsDynamicallyLoaded=false
     drawAs=Image
     imageType=FullColor
     mirroring=NoMirror
     resourceName=None
     resourceObject=/DiscoveryBrowser/TabScreens/Library/MaterialInstances/MI_UI_Library_StorageBar
     tiling=NoTile
     imageSize=<8 8>
     margin=<ltrb 0 0 0 0>
     outlineSettings {
      bUseBrushTransparency=false
      roundingType=HalfHeightRadius
      width=0
      color {colorUseRule=UseColor_Specified, specifiedColor=<rgba 0 0 0 0>}
      cornerRadii=<0 0 0 0>
     }
     tintColor {colorUseRule=UseColor_Specified, specifiedColor=<rgba 1 1 1 1>}
     uVRegion {bIsValid=false, max=<0 0>, min=<0 0>}
    }
   }
   slot {class=GridSlot, properties {layer=1, padding=<ltrb 4 4 4 4>}}
  }
 ]
}
