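# Small legend pin widget for the Discover Library storage UI, showing a two-color dot on a circular obscured backplate.
# A GridPanel layers a minimum-size spacer, the UIKit backplate and a Pin image that uses a storage legend pin material instance.
# PreConstruct creates a dynamic material from the Pin brush, then SetVisuals writes the Color1 and Color2 variables into its color parameters.
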WidgetBlueprint /DiscoveryBrowser/TabScreens/Library/WBP_UI_Discover_Library_StorageLegendPin
graphs [
 EventGraph {nodes [
   Event PreConstruct()
   > n1 = GetDynamicMaterial(self=Get Pin() -> (Pin:Image Object Reference))
   > Set DynamicPinMaterial(DynamicPinMaterial:Material Instance Dynamic Object Reference=@n1) -> (Output_Get:Material Instance Dynamic Object Reference)
   > SetVisuals()
  ]}
 SetVisuals {nodes [
   SetVisuals() {is=title K2Node_FunctionEntry}
   > SetVectorParameterValue(
    self=Get DynamicPinMaterial() -> (DynamicPinMaterial:Material Instance Dynamic Object Reference)
    ParameterName=Color 1
    Value=Get Color1() -> (Color1:Linear Color Structure)
   )
   > SetVectorParameterValue(self=Get DynamicPinMaterial(), ParameterName=Color 2, Value=Get Color2() -> (Color2:Linear Color Structure))
  ]}
]
variables [
 Color1 {type=LinearColor, default=<rgba 0 1 0.3093 1>}
 Color2 {type=LinearColor, default=<rgba 0.9461 0 1 1>}
 DynamicPinMaterial {type=MaterialInstanceDynamic}
]
widgets {
 info {inheritedWidgetCount=0, parentClass=&UserWidget, rootWidgetClass=&GridPanel, widgetCount=4}
 widgets [
  GridPanel_23 {widgetClassPath=&GridPanel, widget=GridPanel_23, properties {columnFill [1], rowFill [1]}}
  Min_Size {
   widgetClassPath=&Spacer
   parent=GridPanel_23
   widget=Min_Size
   properties {size=<12 12>}
   slot {class=GridSlot, properties {layer=-1}}
  }
  Backplate {
   widgetClassPath=&WBP_UIKit_Backplate
   parent=GridPanel_23
   widget=Backplate
   properties {brightness=Obscure, cornerRadius=Circle}
   slot {class=GridSlot}
  }
  Pin {
   widgetClassPath=&Image
   bIsVariable=true
   parent=GridPanel_23
   widget=Pin
   properties {
    visibility=HitTestInvisible
    brush {
     bIsDynamicallyLoaded=false
     drawAs=Image
     imageType=FullColor
     mirroring=NoMirror
     resourceName=None
     resourceObject=/DiscoveryBrowser/TabScreens/Library/MaterialInstances/MI_UI_Library_StorageLegendPin
     tiling=NoTile
     imageSize=<8 8>
     margin=<ltrb 0 0 0 0>
     outlineSettings {
      bUseBrushTransparency=false
      roundingType=HalfHeightRadius
      width=0
      color {colorUseRule=UseColor_Specified, specifiedColor=<rgba 0 0 0 0>}
      cornerRadii=<0 0 0 0>
     }
     tintColor {colorUseRule=UseColor_Specified, specifiedColor=<rgba 1 1 1 1>}
     uVRegion {bIsValid=false, max=<0 0>, min=<0 0>}
    }
   }
   slot {class=GridSlot, properties {layer=1, padding=<ltrb 2 2 2 2>}}
  }
 ]
}
