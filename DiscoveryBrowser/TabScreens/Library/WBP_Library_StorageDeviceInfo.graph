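# Device storage panel for the Discovery Library tab, showing used-of-total space, a storage bar and a Fortnite, Non-Fortnite and Free legend.
# Setter functions format byte counts as SI memory text and feed sizes to the bar.
# Construct initializes the storage device view model when library storage is enabled; design-time PreConstruct previews debug values.
# Non-Fortnite size is content minus Fortnite, floored at the Fortnite size rather than zero.
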WidgetBlueprint /DiscoveryBrowser/TabScreens/Library/WBP_Library_StorageDeviceInfo
graphs [
 EventGraph {nodes [
   n1 = Event PreConstruct()
   > Branch(Condition=@n1.IsDesignTime)
   > SetFortniteSpaceSize!(execute, self, MemorySize:Integer64=Get DebugFortniteMemorySize() -> (DebugFortniteMemorySize:Integer64=0)) -> (then)
   > SetFreeSpaceSize!(execute, self, MemorySize:Integer64=Get DebugFreeMemorySize() -> (DebugFreeMemorySize:Integer64=0)) -> (then)
   n2 = Get DebugFortniteMemorySize() -> (DebugFortniteMemorySize=0)
   n3 = Get DebugTotalContentMemorySize() -> (DebugTotalContentMemorySize:Integer64=0)
   > SetNonFortniteSpaceSize!(
    execute
    self
    MemorySize:Integer64="Max (Integer64)"(
     A:Integer64=integer64 - integer64(A:Integer64=@n3, B:Integer64=@n2) -> (ReturnValue:Integer64) {is=title K2Node_PromotableOperator}
     B:Integer64=@n2
    ) -> (ReturnValue:Integer64=0) {is=title K2Node_CommutativeAssociativeBinaryOperator}
   ) -> (then)
   > SetTotalSpaceSize!(execute, self, MemorySize:Integer64=Get DebugTotalMemorySize() -> (DebugTotalMemorySize:Integer64=0)) -> (then)
   > SetTotalContentSpaceSize!(execute, self, MemorySize:Integer64=Get DebugTotalContentMemorySize() -> (DebugTotalContentMemorySize=0)) -> (then)
   > Set Text(
    Text=GetTotalStorageSpace!(
     self
     TotalContentMemorySize:Integer64=Get DebugTotalContentMemorySize() -> (DebugTotalContentMemorySize=0)
     TotalMemorySize:Integer64=Get DebugTotalMemorySize() -> (DebugTotalMemorySize=0)
    ) -> (MemorySizeTxt:Text)
    self:Rich Text Block Object Reference=Get RT_DeviceStorageValues() -> (RT_DeviceStorageValues:Common Rich Text Block Object Reference)
   )
   Event Construct()
   > n4 = IsLibraryStorageEnabled!(execute) -> (then, ReturnValue:bool)
   > Branch(Condition=@n4)
   > InitializeProperties(self:Fort Library Storage Device View Model Object Reference=Get FortLibraryStorageDeviceViewModel() -> (FortLibraryStorageDeviceViewModel:Fort Library Storage Device View Model Object Reference))
  ]}
 GetNonFortniteSpaceSize {nodes [
   n1 = GetNonFortniteSpaceSize!() -> (then, TotalContentSize:Integer64, FortniteSize:Integer64) {is=title K2Node_FunctionEntry}
   > Return Node!(execute, MemorySize:Integer64="Max (Integer64)"(A=integer64 - integer64(A=@n1, B=@n1.FortniteSize), B=@n1.FortniteSize) -> (ReturnValue=0))
  ]}
 GetTotalStorageSpace {nodes [
   n1 = GetTotalStorageSpace!() -> (then, TotalContentMemorySize:Integer64, TotalMemorySize:Integer64) {is=title K2Node_FunctionEntry}
   > Return Node!(
    execute
    MemorySizeTxt:Text=Format Text!(
     Format="LOCTEXT(\"{TotalContentSpaceSize} <Footnote10>of {TotalSpaceSize} used</>\")"
     TotalContentSpaceSize:Text=Set Memory Format!(NumBytes:Integer64=@n1) -> (FormattedMemory:Text) {is=title K2Node_MacroInstance}
     TotalSpaceSize:Text=Set Memory Format(NumBytes=@n1.TotalMemorySize)
    ) -> (Result)
   )
  ]}
 SetFortniteSpaceSize {nodes [
   n1 = SetFortniteSpaceSize!() -> (then, MemorySize:Integer64) {is=title K2Node_FunctionEntry}
   > SetMemorySize!(
    execute
    self
    MemorySize:Integer64=@n1
    Textfield:Text Object Reference=Get TXT_ContentSize_Value() -> (TXT_ContentSize_Value:Common Text Object Reference)
   ) -> (then)
   > SetFortniteStorageSize!(
    execute
    self:WBP UI Discover Library Storage Bar Object Reference=Get StorageBar() -> (StorageBar:WBP UI Discover Library Storage Bar Object Reference)
    Size:float="To Float (Integer64)"!(InInt:Integer64=Get MemorySize() -> (MemorySize:Integer64=0)) -> (ReturnValue:double)
   ) -> (then)
  ]}
 SetFreeSpaceSize {nodes [
   n1 = SetFreeSpaceSize!() -> (then, MemorySize:Integer64) {is=title K2Node_FunctionEntry}
   > SetMemorySize(MemorySize=@n1, Textfield=Get Txt_FreeSpace_Value() -> (Txt_FreeSpace_Value:Common Text Object Reference))
  ]}
 SetMemoryFormat {nodes [Outputs!(FormattedMemory:Text=AsMemory!(
     NumBytes:Integer64=Inputs!() -> (NumBytes:Integer64)
     UnitStandard:EMemoryUnitStandard Enum=SI
     bUseGrouping:bool=false
     MinimumIntegralDigits:int=1
     MaximumIntegralDigits:int=324
     MinimumFractionalDigits:int=0
     MaximumFractionalDigits:int=2
    ) -> (ReturnValue:Text))]}
 SetMemorySize {nodes [
   n1 = SetMemorySize!() -> (then, MemorySize:Integer64, Textfield:Text Object Reference) {is=title K2Node_FunctionEntry}
   > Set Text(Text=Set Memory Format(NumBytes=@n1), self:Text Object Reference=@n1.Textfield)
  ]}
 SetNonFortniteSpaceSize {nodes [
   n1 = SetNonFortniteSpaceSize!() -> (then, MemorySize:Integer64) {is=title K2Node_FunctionEntry}
   > SetMemorySize(MemorySize=@n1, Textfield=Get Txt_NonFortniteContent_Value() -> (Txt_NonFortniteContent_Value:Common Text Object Reference)) {is=title K2Node_CallFunction}
  ]}
 SetTotalContentSpaceSize {nodes [
   n1 = SetTotalContentSpaceSize!() -> (then, MemorySize:Integer64) {is=title K2Node_FunctionEntry}
   > SetTotalContentStorageSize!(execute, self:WBP UI Discover Library Storage Bar Object Reference=Get StorageBar(), Size:float="To Float (Integer64)"(InInt=@n1)) -> (then)
  ]}
 SetTotalSpaceSize {nodes [
   n1 = SetTotalSpaceSize!() -> (then, MemorySize:Integer64) {is=title K2Node_FunctionEntry}
   > SetTotalDiskStorageSize!(execute, self:WBP UI Discover Library Storage Bar Object Reference=Get StorageBar(), Size:float="To Float (Integer64)"(InInt=@n1)) -> (then)
  ]}
]
variables [
 DebugFortniteMemorySize {type=integer, default=21474836480}
 DebugFreeMemorySize {type=integer, default=361546376}
 DebugTotalContentMemorySize {type=integer, default=987842478080}
 DebugTotalMemorySize {type=integer, default=1000000000000}
]
widgets {
 info {inheritedWidgetCount=0, parentClass=&CommonUserWidget, rootWidgetClass=&Overlay, widgetCount=19}
 widgets [
  Overlay_106 {widgetClassPath=&Overlay, widget=Overlay_106}
  Min_Size {
   widgetClassPath=&Spacer
   parent=Overlay_106
   widget=Min_Size
   properties {size=<416 212>}
   slot {class=OverlaySlot, properties {horizontalAlignment=HAlign_Fill, verticalAlignment=VAlign_Fill}}
  }
  WBP_UIKit_Backplate {
   widgetClassPath=&WBP_UIKit_Backplate
   parent=Overlay_106
   widget=WBP_UIKit_Backplate
   properties {brightness=Dark}
   slot {class=OverlaySlot, properties {horizontalAlignment=HAlign_Fill, verticalAlignment=VAlign_Fill}}
  }
  VB_Content {
   widgetClassPath=&VerticalBox
   parent=Overlay_106
   widget=VB_Content
   slot {class=OverlaySlot, properties {horizontalAlignment=HAlign_Fill, verticalAlignment=VAlign_Fill, padding=<ltrb 24 24 24 24>}}
  }
  TXT_Title {
   widgetClassPath=&CommonTextBlock
   parent=VB_Content
   widget=TXT_Title
   properties {
    mobileFontSizeMultiplier=1.1
    scrollStyle=&TextScrollStyle-Base_C
    text=Device
    textTransformPolicy=ToUpper
    font {
     bForceMonospaced=false
     bMaterialIsStencil=false
     fontMaterial=null
     fontObject=&HeadingNow
     letterSpacing=0
     monospacedWidth=1
     size=15.75
     skewAmount=0
     typefaceFontName=75 Medium
     outlineSettings {
      bApplyOutlineToDropShadows=false
      bMiteredCorners=false
      bSeparateFillAlpha=false
      outlineMaterial=null
      outlineSize=0
      outlineColor=<rgba 0 0 0 1>
     }
    }
    shadowOffset=<0 0>
   }
   slot {class=VerticalBoxSlot, properties {verticalAlignment=VAlign_Top, padding=<ltrb 0 0 0 16>}}
  }
  RT_DeviceStorageValues {
   widgetClassPath=&CommonRichTextBlock
   bIsVariable=true
   parent=VB_Content
   widget=RT_DeviceStorageValues
   properties {
    applyLineHeightToBottomLine=false
    inlineIconDisplayMode=TextOnly
    mobileTextBlockScale=1.25
    scrollStyle=&TextScrollStyle-Base_C
    text="{0} GB <Footnote10>of {0} GB used</>"
    textStyleSet=/DiscoveryBrowser/TabScreens/Library/TextStyles/DT_Discover_Library_Storage_TextStyles
    defaultTextStyleOverride {
     overflowPolicy=Clip
     transformPolicy=None
     colorAndOpacity {colorUseRule=UseColor_Specified, specifiedColor=<rgba 1 1 1 1>}
     font {
      bForceMonospaced=false
      bMaterialIsStencil=false
      fontMaterial=null
      fontObject=null
      letterSpacing=0
      monospacedWidth=1
      size=24
      skewAmount=0
      typefaceFontName=None
      outlineSettings {
       bApplyOutlineToDropShadows=false
       bMiteredCorners=false
       bSeparateFillAlpha=false
       outlineMaterial=null
       outlineSize=0
       outlineColor=<rgba 0 0 0 1>
      }
     }
     highlightColor {colorUseRule=UseColor_Specified, specifiedColor=<rgba 0 0 0 1>}
     highlightShape {
      bIsDynamicallyLoaded=false
      drawAs=Image
      imageType=NoImage
      mirroring=NoMirror
      resourceName=None
      resourceObject=null
      tiling=NoTile
      imageSize=<32 32>
      margin=<ltrb 0 0 0 0>
      outlineSettings {
       bUseBrushTransparency=false
       roundingType=HalfHeightRadius
       width=0
       color {colorUseRule=UseColor_Specified, specifiedColor=<rgba 0 0 0 0>}
       cornerRadii=<0 0 0 0>
      }
      tintColor {colorUseRule=UseColor_Specified, specifiedColor=<rgba 1 1 1 1>}
      uVRegion {bIsValid=false, max=<0 0>, min=<0 0>}
     }
     selectedBackgroundColor {colorUseRule=UseColor_Foreground, specifiedColor=<rgba 1 0 1 1>}
     shadowColorAndOpacity=<rgba 0 0 0 1>
     shadowOffset=<0 0>
     strikeBrush {
      bIsDynamicallyLoaded=false
      drawAs=Image
      imageType=NoImage
      mirroring=NoMirror
      resourceName=None
      resourceObject=null
      tiling=NoTile
      imageSize=<32 32>
      margin=<ltrb 0 0 0 0>
      outlineSettings {
       bUseBrushTransparency=false
       roundingType=HalfHeightRadius
       width=0
       color {colorUseRule=UseColor_Specified, specifiedColor=<rgba 0 0 0 0>}
       cornerRadii=<0 0 0 0>
      }
      tintColor {colorUseRule=UseColor_Specified, specifiedColor=<rgba 1 1 1 1>}
      uVRegion {bIsValid=false, max=<0 0>, min=<0 0>}
     }
     underlineBrush {
      bIsDynamicallyLoaded=false
      drawAs=Image
      imageType=NoImage
      mirroring=NoMirror
      resourceName=None
      resourceObject=null
      tiling=NoTile
      imageSize=<32 32>
      margin=<ltrb 0 0 0 0>
      outlineSettings {
       bUseBrushTransparency=false
       roundingType=HalfHeightRadius
       width=0
       color {colorUseRule=UseColor_Specified, specifiedColor=<rgba 0 0 0 0>}
       cornerRadii=<0 0 0 0>
      }
      tintColor {colorUseRule=UseColor_Specified, specifiedColor=<rgba 1 1 1 1>}
      uVRegion {bIsValid=false, max=<0 0>, min=<0 0>}
     }
    }
   }
   slot {class=VerticalBoxSlot, properties {verticalAlignment=VAlign_Bottom, padding=<ltrb 0 0 0 4>}}
  }
  StorageBar {
   widgetClassPath=&/WBP_UI_Discover_Library_StorageBar
   bIsVariable=true
   parent=VB_Content
   widget=StorageBar
   slot {class=VerticalBoxSlot, properties {verticalAlignment=VAlign_Top, padding=<ltrb 0 0 0 12>}}
  }
  OV_ListItem_Fortnite {
   widgetClassPath=&HorizontalBox
   parent=VB_Content
   widget=OV_ListItem_Fortnite
   slot {class=VerticalBoxSlot, properties {verticalAlignment=VAlign_Top, padding=<ltrb 4 0 4 8>}}
  }
  WBP_UI_Discover_Library_StorageLegendPin {
   widgetClassPath=&/WBP_UI_Discover_Library_StorageLegendPin
   parent=OV_ListItem_Fortnite
   widget=WBP_UI_Discover_Library_StorageLegendPin
   properties {color1=<rgba 0.9301 1 0.0103 1>, color2=<rgba 0.9301 1 0.0103 1>}
   slot {class=HorizontalBoxSlot, properties {horizontalAlignment=HAlign_Left, verticalAlignment=VAlign_Center, padding=<ltrb 0 0 8 0>}}
  }
  Txt_FortniteContent {
   widgetClassPath=&CommonTextBlock
   parent=OV_ListItem_Fortnite
   widget=Txt_FortniteContent
   properties {
    mobileFontSizeMultiplier=1.25
    scrollStyle=&TextScrollStyle-Base_C
    text=Fortnite Content
    colorAndOpacity {colorUseRule=UseColor_Specified, specifiedColor=<rgba 0.5776 0.5776 0.5776 1>}
    font {
     bForceMonospaced=false
     bMaterialIsStencil=false
     fontMaterial=null
     fontObject=&HeadingNow
     letterSpacing=0
     monospacedWidth=1
     size=11.25
     skewAmount=0
     typefaceFontName=74 Regular
     outlineSettings {
      bApplyOutlineToDropShadows=false
      bMiteredCorners=false
      bSeparateFillAlpha=false
      outlineMaterial=null
      outlineSize=0
      outlineColor=<rgba 0 0 0 1>
     }
    }
    shadowOffset=<0 0>
   }
   slot {class=HorizontalBoxSlot, properties {verticalAlignment=VAlign_Center, size {sizeRule=Fill, value=1}}}
  }
  TXT_ContentSize_Value {
   widgetClassPath=&CommonTextBlock
   bIsVariable=true
   parent=OV_ListItem_Fortnite
   widget=TXT_ContentSize_Value
   properties {
    justification=Right
    mobileFontSizeMultiplier=1.25
    scrollStyle=&TextScrollStyle-Base_C
    text="{0} GB"
    font {
     bForceMonospaced=false
     bMaterialIsStencil=false
     fontMaterial=null
     fontObject=&HeadingNow
     letterSpacing=0
     monospacedWidth=1
     size=11.25
     skewAmount=0
     typefaceFontName=74 Regular
     outlineSettings {
      bApplyOutlineToDropShadows=false
      bMiteredCorners=false
      bSeparateFillAlpha=false
      outlineMaterial=null
      outlineSize=0
      outlineColor=<rgba 0 0 0 1>
     }
    }
    shadowOffset=<0 0>
   }
   slot {class=HorizontalBoxSlot, properties {horizontalAlignment=HAlign_Right, verticalAlignment=VAlign_Center, padding=<ltrb 8 0 0 0>}}
  }
  OV_ListItem_Other {
   widgetClassPath=&HorizontalBox
   parent=VB_Content
   widget=OV_ListItem_Other
   slot {class=VerticalBoxSlot, properties {verticalAlignment=VAlign_Top, padding=<ltrb 4 0 4 8>}}
  }
  WBP_UI_Discover_Library_StorageLegendPin_1 {
   widgetClassPath=&WBP_UI_Discover_Library_StorageLegendPin
   parent=OV_ListItem_Other
   widget=WBP_UI_Discover_Library_StorageLegendPin_1
   properties {color1=<rgba 0 0.0976 0.521 1>, color2=<rgba 0.0482 0.3467 0.9473 1>}
   slot {class=HorizontalBoxSlot, properties {horizontalAlignment=HAlign_Left, verticalAlignment=VAlign_Center, padding=<ltrb 0 0 8 0>}}
  }
  Txt_NonFortniteContent {
   widgetClassPath=&CommonTextBlock
   parent=OV_ListItem_Other
   widget=Txt_NonFortniteContent
   properties {
    mobileFontSizeMultiplier=1.25
    scrollStyle=&TextScrollStyle-Base_C
    text=Non-Fortnite
    colorAndOpacity {colorUseRule=UseColor_Specified, specifiedColor=<rgba 0.5776 0.5776 0.5776 1>}
    font {
     bForceMonospaced=false
     bMaterialIsStencil=false
     fontMaterial=null
     fontObject=&HeadingNow
     letterSpacing=0
     monospacedWidth=1
     size=11.25
     skewAmount=0
     typefaceFontName=74 Regular
     outlineSettings {
      bApplyOutlineToDropShadows=false
      bMiteredCorners=false
      bSeparateFillAlpha=false
      outlineMaterial=null
      outlineSize=0
      outlineColor=<rgba 0 0 0 1>
     }
    }
    shadowOffset=<0 0>
   }
   slot {class=HorizontalBoxSlot, properties {verticalAlignment=VAlign_Center, size {sizeRule=Fill, value=1}}}
  }
  Txt_NonFortniteContent_Value {
   widgetClassPath=&CommonTextBlock
   bIsVariable=true
   parent=OV_ListItem_Other
   widget=Txt_NonFortniteContent_Value
   properties {
    justification=Right
    mobileFontSizeMultiplier=1.25
    scrollStyle=&TextScrollStyle-Base_C
    text="{0} GB"
    font {
     bForceMonospaced=false
     bMaterialIsStencil=false
     fontMaterial=null
     fontObject=&HeadingNow
     letterSpacing=0
     monospacedWidth=1
     size=11.25
     skewAmount=0
     typefaceFontName=74 Regular
     outlineSettings {
      bApplyOutlineToDropShadows=false
      bMiteredCorners=false
      bSeparateFillAlpha=false
      outlineMaterial=null
      outlineSize=0
      outlineColor=<rgba 0 0 0 1>
     }
    }
    shadowOffset=<0 0>
   }
   slot {class=HorizontalBoxSlot, properties {horizontalAlignment=HAlign_Right, verticalAlignment=VAlign_Center, padding=<ltrb 8 0 0 0>}}
  }
  OV_ListItem_Free {
   widgetClassPath=&HorizontalBox
   parent=VB_Content
   widget=OV_ListItem_Free
   slot {class=VerticalBoxSlot, properties {verticalAlignment=VAlign_Top, padding=<ltrb 4 0 4 0>}}
  }
  WBP_UI_Discover_Library_StorageLegendPin_2 {
   widgetClassPath=&WBP_UI_Discover_Library_StorageLegendPin
   parent=OV_ListItem_Free
   widget=WBP_UI_Discover_Library_StorageLegendPin_2
   properties {color1=<rgba 0.0666 0.0666 0.0666 1>, color2=<rgba 0.0666 0.0666 0.0666 1>}
   slot {class=HorizontalBoxSlot, properties {horizontalAlignment=HAlign_Left, verticalAlignment=VAlign_Center, padding=<ltrb 0 0 8 0>}}
  }
  Txt_FreeSpace {
   widgetClassPath=&CommonTextBlock
   parent=OV_ListItem_Free
   widget=Txt_FreeSpace
   properties {
    mobileFontSizeMultiplier=1.25
    scrollStyle=&TextScrollStyle-Base_C
    text=Free
    colorAndOpacity {colorUseRule=UseColor_Specified, specifiedColor=<rgba 0.5776 0.5776 0.5776 1>}
    font {
     bForceMonospaced=false
     bMaterialIsStencil=false
     fontMaterial=null
     fontObject=&HeadingNow
     letterSpacing=0
     monospacedWidth=1
     size=11.25
     skewAmount=0
     typefaceFontName=74 Regular
     outlineSettings {
      bApplyOutlineToDropShadows=false
      bMiteredCorners=false
      bSeparateFillAlpha=false
      outlineMaterial=null
      outlineSize=0
      outlineColor=<rgba 0 0 0 1>
     }
    }
    shadowOffset=<0 0>
   }
   slot {class=HorizontalBoxSlot, properties {verticalAlignment=VAlign_Center, size {sizeRule=Fill, value=1}}}
  }
  Txt_FreeSpace_Value {
   widgetClassPath=&CommonTextBlock
   bIsVariable=true
   parent=OV_ListItem_Free
   widget=Txt_FreeSpace_Value
   properties {
    justification=Right
    mobileFontSizeMultiplier=1.25
    scrollStyle=&TextScrollStyle-Base_C
    text="{0} GB"
    font {
     bForceMonospaced=false
     bMaterialIsStencil=false
     fontMaterial=null
     fontObject=&HeadingNow
     letterSpacing=0
     monospacedWidth=1
     size=11.25
     skewAmount=0
     typefaceFontName=74 Regular
     outlineSettings {
      bApplyOutlineToDropShadows=false
      bMiteredCorners=false
      bSeparateFillAlpha=false
      outlineMaterial=null
      outlineSize=0
      outlineColor=<rgba 0 0 0 1>
     }
    }
    shadowOffset=<0 0>
   }
   slot {class=HorizontalBoxSlot, properties {horizontalAlignment=HAlign_Right, verticalAlignment=VAlign_Center, padding=<ltrb 8 0 0 0>}}
  }
 ]
}
