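# Confirmation modal for removing downloaded Library content, built on FortRemoveConfirmationModal with a UIKit dialog shell.
# PreConstruct supplies the dialog viewmodel and binds button clicks; the storage item ID initializes the storage item viewmodel.
# Button 0 calls RemoveDownload on that viewmodel, any other cancels, and both deactivate the modal and fire analytics.
# Slotted text shows name, freed space formatted in SI units, and notice that removal happens next launch.
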WidgetBlueprint /FortniteCoreUI/Library/Dialogs/WBP_Library_RemoveConfirmation_Modal
graphs [
 EventGraph {nodes [
   n1 = OnButtonClicked!() -> (OutputDelegate, then, ButtonIndex:int) {is=title K2Node_CustomEvent}
   > n2 = Switch on Int!(execute:Exec, Selection:int=@n1.ButtonIndex) -> (Default:Exec, 0:Exec)
   RemoveDownload(
    execute=@n2.0
    self:Fort Library Storage Item View Model Object Reference=Get FortLibraryStorageItemViewModel() -> (FortLibraryStorageItemViewModel:Fort Library Storage Item View Model Object Reference)
   )
   > DeactivateWidget()
   > n3 = Cast To FortPlayerController(Object=GetOwningPlayer())
   > "Fire Event Element Interacted (Non-Localized)"!(
    execute
    FortPC:Fort Player Controller Object Reference=@n3
    ElementName:String=Button_Remove
    AdditionalInfo:Flow Analytics Additional Info Structure
   ) -> (then)
   DeactivateWidget(execute=@n2.Default)
   > n4 = Cast To FortPlayerController(Object=GetOwningPlayer())
   > "Fire Event Element Interacted (Non-Localized)"(FortPC=@n4, ElementName=Button_Cancel)
   Event PreConstruct()
   > n5 = Construct Object from Class(Class=/FortniteCoreUI/Library/Dialogs/Data/DialogVM_Library_RemoveConfirmation.DialogVM_Library_RemoveConfirmation_C) -> (ReturnValue:Dialog VM Library Remove Confirmation Object Reference)
   n6 = Get Dialog_Internal() -> (Dialog_Internal:WBP UIKit Dialog Internal Object Reference)
   > SetDialogData!(execute, self:WBP UIKit Dialog Internal Object Reference=@n6, Data:UI Kit Dialog Viewmodel Object Reference=@n5) -> (then)
   > Bind Event to Dialog Button Clicked Dispatcher(self:WBP UIKit Dialog Internal Object Reference=@n6, Delegate=@n1) {is=title K2Node_AddDelegate}
   Event On Activated()
   > Fire Event Main View Opened!(execute, View:Widget Object Reference=Self-Reference(), AdditionalInfo:Flow Analytics Additional Info Structure, bIsOverlay:bool=false) -> (then)
   n7 = Event BP_OnSetStorageItemId!() -> (OutputDelegate, then, InStorageItemId:String) {is=title K2Node_Event}
   > InitializeWithStorageItemId!(
    execute
    self:Fort Library Storage Item View Model Object Reference=Get FortLibraryStorageItemViewModel()
    InStorageItemId:String=@n7.InStorageItemId
   ) -> (then)
  ]}
 OnRemoveClicked {nodes [OnRemoveClicked() {is=title K2Node_FunctionEntry}]}
 SetMemoryFormat {nodes [Outputs!(FormattedMemory:Text=AsMemory!(
     NumBytes:Integer64=Inputs!() -> (NumBytes:Integer64)
     UnitStandard:EMemoryUnitStandard Enum=SI
     bUseGrouping:bool=false
     MinimumIntegralDigits:int=1
     MaximumIntegralDigits:int=324
     MinimumFractionalDigits:int=0
     MaximumFractionalDigits:int=2
    ) -> (ReturnValue:Text))]}
 SetMemoryText {nodes [
   n1 = SetMemoryText!() -> (then, MemorySize:Integer64) {is=title K2Node_FunctionEntry}
   > Set Text(
    Text=Format Text!(
     Format="LOCTEXT(\"This will free up {ContentSize} on your device.\")"
     ContentSize:Text=Set Memory Format!(NumBytes:Integer64=@n1) -> (FormattedMemory:Text) {is=title K2Node_MacroInstance}
    ) -> (Result)
    self=Get TXT_ContentSize() -> (TXT_ContentSize:Common Text Object Reference)
   )
  ]}
]
variables [OnRemoveClicked {}]
widgets {
 info {
  inheritedWidgetCount=0
  namedSlotCount=1
  parentClass=/Script/FortniteUI.FortRemoveConfirmationModal
  rootWidgetClass=/FortUILibrary/Systems/Dialog/WBP_UIKit_Dialog_Internal.WBP_UIKit_Dialog_Internal_C
  widgetCount=6
 }
 widgets [
  Dialog_Internal {
   widgetClassPath=/FortUILibrary/Systems/Dialog/WBP_UIKit_Dialog_Internal
   bIsVariable=true
   widget=Dialog_Internal
   properties {
    useText=None
    backgroundProperties {
     material=/FortUILibrary/UIKit/Dialogs/MaterialInstances/MI_UI_DialogBackground
     image Size=<32 32>
     placement {
      horizontal Alignment=HAlign_Fill
      vertical Alignment=VAlign_Fill
      grid {column=0, column Span=2, layer=1, row=0, row Span=0, nudge=<0 0>}
      padding=<ltrb 0 0 0 0>
     }
    }
    buttons {
     each Button Horizontal Fill=true
     each Button Min Width=200
     each Button Size=Medium
     focuse Button=0
     space Between Buttons=30
     cTA Buttons Types [Loud,Regular]
     group Placement {
      horizontal Alignment=HAlign_Fill
      vertical Alignment=VAlign_Bottom
      grid {column=1, column Span=0, layer=0, row=2, row Span=0, nudge=<0 0>}
      padding=<ltrb 80 64 80 48>
     }
    }
    dialogSIze {
     height Override {enable=false, value=800}
     max Aspect Ratio {enable=false, value=0}
     max Desired Height {enable=true, value=800}
     max Desired Width {enable=false, value=0}
     min Aspect Ratio {enable=false, value=0}
     min Desired Height {enable=false, value=0}
     min Desired Width {enable=false, value=0}
     width Override {enable=true, value=864}
    }
    textProperties {
     auto Wrap Text=true
     justification=Left
     line Height Percentage=1.2
     mobile Font Size Multiplier=1.3
     transform Policy=None
     wrap Text At=0
     color {colorUseRule=UseColor_Specified, specifiedColor=<rgba 1 1 1 1>}
     font {
      bForceMonospaced=false
      bMaterialIsStencil=false
      fontMaterial=null
      fontObject=&HeadingNow
      letterSpacing=0
      monospacedWidth=1
      size=22.5
      skewAmount=0
      typefaceFontName=74 Regular
      outlineSettings {
       bApplyOutlineToDropShadows=false
       bMiteredCorners=false
       bSeparateFillAlpha=false
       outlineMaterial=null
       outlineSize=0
       outlineColor=<rgba 0 0 0 1>
      }
     }
     placement {
      horizontal Alignment=HAlign_Fill
      vertical Alignment=VAlign_Fill
      grid {column=1, column Span=0, layer=0, row=1, row Span=0, nudge=<0 0>}
      padding=<ltrb 80 80 80 80>
     }
    }
   }
  }
  OV_NewContent {widgetClassPath=&Overlay, namedSlotHost=Dialog_Internal, widget=OV_NewContent}
  VB_Paddings {
   widgetClassPath=&VerticalBox
   parent=OV_NewContent
   widget=VB_Paddings
   slot {class=OverlaySlot, properties {horizontalAlignment=HAlign_Fill, verticalAlignment=VAlign_Fill, padding=<ltrb 80 0 80 0>}}
  }
  TXT_ContentName {
   widgetClassPath=&CommonTextBlock
   bIsVariable=true
   parent=VB_Paddings
   widget=TXT_ContentName
   properties {
    autoWrapText=true
    lineHeightPercentage=1.15
    text="Are you sure you want to remove {ContentName}? "
    colorAndOpacity {colorUseRule=UseColor_Specified, specifiedColor=<rgba 1 1 1 100>}
    font {
     bForceMonospaced=false
     bMaterialIsStencil=false
     fontMaterial=null
     fontObject=&HeadingNow
     letterSpacing=0
     monospacedWidth=1
     size=22.5
     skewAmount=0
     typefaceFontName=74 Regular
     outlineSettings {
      bApplyOutlineToDropShadows=false
      bMiteredCorners=false
      bSeparateFillAlpha=false
      outlineMaterial=null
      outlineSize=0
      outlineColor=<rgba 0 0 0 1>
     }
    }
    shadowOffset=<0 0>
   }
   slot {class=VerticalBoxSlot, properties {verticalAlignment=VAlign_Top, padding=<ltrb 0 0 0 40>}}
  }
  TXT_ContentSize {
   widgetClassPath=&CommonTextBlock
   bIsVariable=true
   parent=VB_Paddings
   widget=TXT_ContentSize
   properties {
    autoWrapText=true
    lineHeightPercentage=1.15
    text="This will free up {ContentSize} on your device."
    colorAndOpacity {colorUseRule=UseColor_Specified, specifiedColor=<rgba 1 1 1 100>}
    font {
     bForceMonospaced=false
     bMaterialIsStencil=false
     fontMaterial=null
     fontObject=&HeadingNow
     letterSpacing=0
     monospacedWidth=1
     size=22.5
     skewAmount=0
     typefaceFontName=74 Regular
     outlineSettings {
      bApplyOutlineToDropShadows=false
      bMiteredCorners=false
      bSeparateFillAlpha=false
      outlineMaterial=null
      outlineSize=0
      outlineColor=<rgba 0 0 0 1>
     }
    }
    shadowOffset=<0 0>
   }
   slot {class=VerticalBoxSlot, properties {verticalAlignment=VAlign_Top, padding=<ltrb 0 0 0 40>}}
  }
  TXT_ContentSchedule {
   widgetClassPath=&CommonTextBlock
   bIsVariable=true
   parent=VB_Paddings
   widget=TXT_ContentSchedule
   properties {
    autoWrapText=true
    lineHeightPercentage=1.15
    text=We'll remove this content next time you start Fortnite.
    colorAndOpacity {colorUseRule=UseColor_Specified, specifiedColor=<rgba 1 1 1 100>}
    font {
     bForceMonospaced=false
     bMaterialIsStencil=false
     fontMaterial=null
     fontObject=&HeadingNow
     letterSpacing=0
     monospacedWidth=1
     size=22.5
     skewAmount=0
     typefaceFontName=74 Regular
     outlineSettings {
      bApplyOutlineToDropShadows=false
      bMiteredCorners=false
      bSeparateFillAlpha=false
      outlineMaterial=null
      outlineSize=0
      outlineColor=<rgba 0 0 0 1>
     }
    }
    shadowOffset=<0 0>
   }
   slot {class=VerticalBoxSlot, properties {verticalAlignment=VAlign_Top}}
  }
 ]
}
