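# List entry widget for the Library storage screen, showing one content pack's thumbnail, description, download status and progress from a FortLibraryStorageItemViewModel.
# Helper functions map download status to text, visibility and thumbnail material parameters for progress and removal.
# SetDownloadStatus rebuilds a status-dependent context menu for download, update, pause, resume, cancel and removal, which opens a confirmation modal.
# The menu opens by click, right-click or a gamepad bottom-bar decoy button shown on hover.
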WidgetBlueprint /DiscoveryBrowser/TabScreens/Library/WBP_Library_StorageItemEntry
graphs [
 CM_CancelDownload {nodes [
   CM_CancelDownload!() -> (then, InteractionNameID:Name) {is=title K2Node_FunctionEntry}
   > CancelDownload(self:Fort Library Storage Item View Model Object Reference=Get FortLibraryStorageItemViewModel() -> (FortLibraryStorageItemViewModel:Fort Library Storage Item View Model Object Reference))
   > CloseContextMenu()
  ]}
 CM_DownloadNow {nodes [
   CM_DownloadNow!() -> (then, InteractionNameID:Name) {is=title K2Node_FunctionEntry}
   > Download(self:Fort Library Storage Item View Model Object Reference=Get FortLibraryStorageItemViewModel())
   > CloseContextMenu()
  ]}
 CM_PauseDownload {nodes [
   CM_PauseDownload!() -> (then, InteractionNameID:Name) {is=title K2Node_FunctionEntry}
   > PauseDownload(self:Fort Library Storage Item View Model Object Reference=Get FortLibraryStorageItemViewModel())
   > CloseContextMenu()
  ]}
 CM_RemoveContent {nodes [
   CM_RemoveContent!() -> (then, InteractionNameID:Name) {is=title K2Node_FunctionEntry}
   > n1 = Create Widget!(
    execute:Exec
    Class:User Widget Class Reference=/FortniteCoreUI/Library/Dialogs/WBP_Library_RemoveConfirmation_Modal.WBP_Library_RemoveConfirmation_Modal_C
    OwningPlayer:Player Controller Object Reference=GetOwningPlayer()
    UIKitDialogViewModel:UI Kit Dialog Viewmodel Object Reference
   ) -> (then:Exec, ReturnValue:WBP Library Remove Confirmation Modal Object Reference)
   > TryPushWidgetAsModalPanel!(execute, PushWidget:User Widget Object Reference=@n1) -> (then)
   > CloseContextMenu()
   > SetStorageItemId!(
    execute
    self:Fort Remove Confirmation Modal Object Reference=@n1
    InStorageItemId:String=Get StorageItemId!(self:Fort Library Storage Item View Model Object Reference=Get FortLibraryStorageItemViewModel()) -> (StorageItemId:String)
   ) -> (then)
  ]}
 CM_ResumeDownload {nodes [
   CM_ResumeDownload!() -> (then, InteractionNameID:Name) {is=title K2Node_FunctionEntry}
   > ResumeDownload(self:Fort Library Storage Item View Model Object Reference=Get FortLibraryStorageItemViewModel())
   > CloseContextMenu()
  ]}
 CM_UpdateNow {nodes [
   CM_UpdateNow!() -> (then, InteractionNameID:Name) {is=title K2Node_FunctionEntry}
   > Download(self=Get FortLibraryStorageItemViewModel())
   > CloseContextMenu()
  ]}
 CloseContextMenu {nodes [
   CloseContextMenu() {is=title K2Node_FunctionEntry}
   > Close(self:Menu Anchor Object Reference=Get Anchor!(self:UIKit Context Menu Anchor Object Reference=Get ContextMenu() -> (ContextMenu:WBP UIKit Context Menu Anchor Button Object Reference)) -> (Anchor:Menu Anchor Object Reference))
  ]}
 EventGraph {nodes [
   "On Clicked (WBP_Discover_Button_Generic)"!() -> (OutputDelegate, then, Button:Common Button Base Object Reference) {is=title K2Node_ComponentBoundEvent}
   > ToggleOpen!(execute, self:Menu Anchor Object Reference=Get Anchor(self=Get ContextMenu()), bFocusOnOpen:bool=false) -> (then)
   > Branch(Condition=IsUsingTouch(PlayerController=GetOwningPlayer()))
   > Call On Button Selected Touch!(execute:Exec, self:WBP Library Storage Item Entry Object Reference, ListEntry:Common User Widget Object Reference=Self-Reference()) -> (then:Exec) {is=title K2Node_CallDelegate}
   "On Clicked (WBP_BottomBarHackButton_ContextMenu)"!() -> (OutputDelegate, then, Button:Common Button Base Object Reference) {is=title K2Node_ComponentBoundEvent}
   > ToggleOpen(self=Get Anchor(self=Get ContextMenu()), bFocusOnOpen=false)
   HandleOnHoverChanged!() -> (OutputDelegate, then, Button:Common Button Base Object Reference) {is=title K2Node_CustomEvent}
   > UpdateBottomBarButtonsVisibility!(
    execute
    self
    NewInputType:ECommonInputType Enum=GetCurrentInputType(self=Get CommonInputSubsystem())
    bIsHovered:bool=IsHovered(self=Get WBP_Discover_Button_Generic() -> (WBP_Discover_Button_Generic:WBP Discover Button Generic Object Reference))
   ) -> (then)
   n1 = HandleInputMethodChanged!() -> (OutputDelegate, then, bNewInputType:ECommonInputType Enum) {is=title K2Node_CustomEvent}
   > UpdateBottomBarButtonsVisibility(NewInputType=@n1.bNewInputType, bIsHovered=IsHovered(self=Get WBP_Discover_Button_Generic()))
   Event PreConstruct()
   > n2 = GetDynamicMaterial(self=Get Img_Thumbnail() -> (Img_Thumbnail:Image Object Reference))
   > Set ThumbnailMaterial(ThumbnailMaterial:Material Instance Dynamic Object Reference=@n2) -> (Output_Get:Material Instance Dynamic Object Reference)
   Event Construct()
   > SetupContextMenu()
   > SetHideInputAction!(execute, self:Common Button Base Object Reference=Get WBP_Discover_Button_Generic(), bInHideInputAction:bool=true) -> (then)
   > Bind Event to On Input Method Changed(self=Get CommonInputSubsystem(), Delegate=Create Event())
   > HandleInputMethodChanged!(execute, self, bNewInputType:ECommonInputType Enum=GetCurrentInputType(self=Get CommonInputSubsystem())) -> (then) {is=title K2Node_CallFunction}
   n3 = Create Event()
   n4 = Get WBP_Discover_Button_Generic()
   > Bind Event to On Hovered(self:Common Button Base Object Reference=@n4, Delegate=@n3) {is=title K2Node_AddDelegate}
   > Bind Event to On Unhovered(self:Common Button Base Object Reference=@n4, Delegate=@n3) {is=title K2Node_AddDelegate}
   Event Destruct()
   > Unbind Event from On Input Method Changed(self=Get CommonInputSubsystem(), Delegate=@n1)
   n5 = Get WBP_Discover_Button_Generic()
   > Unbind Event from On Hovered(self:Common Button Base Object Reference=@n5, Delegate=@n3) {is=title K2Node_RemoveDelegate}
   > Unbind Event from On Unhovered(self:Common Button Base Object Reference=@n5, Delegate=@n3) {is=title K2Node_RemoveDelegate}
   n6 = Event OnListItemObjectSet()
   > n7 = Cast To FortLibraryStorageItemViewModel!(execute, Object=@n6.ListItemObject) -> (then, CastFailed, AsFort Library Storage Item View Model:Fort Library Storage Item View Model Object Reference) {is=title K2Node_DynamicCast}
   > n8 = Set FortLibraryStorageItemViewModel(FortLibraryStorageItemViewModel:Fort Library Storage Item View Model Object Reference=@n7) -> (Output_Get:Fort Library Storage Item View Model Object Reference)
   > SetDownloadStatus!(
    execute
    self
    DownloadStatus:ELibraryStorageItemDownloadStatus Enum=Get DownloadStatus!(self:Fort Library Storage Item View Model Object Reference=@n8) -> (DownloadStatus:ELibraryStorageItemDownloadStatus Enum=NotDownloaded)
   ) -> (then)
  ]}
 GetDividerVisibility {nodes [
   n1 = GetDividerVisibility!() -> (then, Core:bool=~) {is=title K2Node_FunctionEntry}
   > Return Node!(execute, Visibility:ESlateVisibility Enum=Select(Option 0=Collapsed, Option 1=SelfHitTestInvisible, Index=@n1))
  ]}
 GetDownloadProgressValue {nodes [
   n1 = GetDownloadProgressValue!() -> (then, DownloadProgressInt:int, DownloadStatus:ELibraryStorageItemDownloadStatus Enum=NotDownloaded) {is=title K2Node_FunctionEntry}
   n2 = float * float(A="To Float (Integer)"(InInt=@n1), B=0.01) {is=title K2Node_PromotableOperator}
   n3 = Select!(
    NotDownloaded:bool=false
    Downloaded:bool=true
    Downloading:bool=false
    UpdateRequired:bool=false
    Updating:bool=false
    Queued:bool=false
    Paused:bool=false
    PendingRemoval:bool=false
    Removing:bool=true
    Index:ELibraryStorageItemDownloadStatus Enum=@n1.DownloadStatus
   ) -> (ReturnValue:bool)
   > Return Node!(
    execute
    DownloadProgress:double="Clamp (Float)"(Value=Select!(Option 0:double=@n2, Option 1:double=1, Index:bool=@n3) -> (ReturnValue:double), Min=0, Max=1)
   )
  ]}
 GetInteractionIndex {nodes [
   n1 = GetInteractionIndex!() -> (then, DownloadStatus:ELibraryStorageItemDownloadStatus Enum=NotDownloaded) {is=title K2Node_FunctionEntry}
   > Return Node!(
    execute
    Index:int=Select(
     Option 0:int=0
     Option 1:int=1
     Index="Equal (Enum)"(A:ELibraryStorageItemDownloadStatus Enum=@n1, B:ELibraryStorageItemDownloadStatus Enum=Removing)
    ) -> (ReturnValue:int)
   )
  ]}
 GetInteractionVisibility {nodes [
   n1 = GetInteractionVisibility!() -> (then, Core:bool=~) {is=title K2Node_FunctionEntry}
   > Return Node!(
    execute
    Visibility:ESlateVisibility Enum=Select(Option 0:ESlateVisibility Enum=SelfHitTestInvisible, Option 1:ESlateVisibility Enum=Hidden, Index=@n1) -> (ReturnValue:ESlateVisibility Enum)
   )
  ]}
 GetPauseIconVisibility {nodes [
   n1 = GetPauseIconVisibility!() -> (then, DownloadStatus:ELibraryStorageItemDownloadStatus Enum=NotDownloaded) {is=title K2Node_FunctionEntry}
   > Return Node(Visibility=Select(Option 0=Collapsed, Option 1=SelfHitTestInvisible, Index="Equal (Enum)"(A=@n1, B=Paused)))
  ]}
 GetRemovingValue {nodes [
   n1 = GetRemovingValue!() -> (then, DownloadStatus:ELibraryStorageItemDownloadStatus Enum=NotDownloaded) {is=title K2Node_FunctionEntry}
   > Return Node!(execute, RemovingValue:double=Select(Option 0:float=0, Option 1:float=1, Index="Equal (Enum)"(A=@n1, B=Removing)) -> (ReturnValue:float))
  ]}
 GetStatusText {nodes [
   n1 = GetStatusText!() -> (then, QueuePosition:int, DownloadStatus:ELibraryStorageItemDownloadStatus Enum=NotDownloaded) {is=title K2Node_FunctionEntry}
   n2 = "To Text (Integer)"(Value=@n1)
   > Return Node!(
    execute
    Text:Text=Select!(
     NotDownloaded:Text="LOCTEXT(\"Not Downloaded\")"
     Downloaded:Text="LOCTEXT(\"Installed\")"
     Downloading:Text="LOCTEXT(\"Downloading\")"
     UpdateRequired:Text="LOCTEXT(\"Update Required\")"
     Updating:Text="LOCTEXT(\"Updating\")"
     Queued:Text=Format Text!(Format="LOCTEXT(\"In Queue ({QueueNumber})\")", QueueNumber:Text=@n2) -> (Result)
     Paused:Text="LOCTEXT(\"Paused\")"
     PendingRemoval:Text
     Removing:Text="LOCTEXT(\"Removing\")"
     Index:ELibraryStorageItemDownloadStatus Enum=@n1.DownloadStatus
    ) -> (ReturnValue:Text)
   )
  ]}
 GetSubStatusText {nodes [
   n1 = GetSubStatusText!() -> (
    then
    StorageItemSize:Integer64
    DownloadSpaceSize:Integer64
    DownloadProgress:int
    DownloadStatus:ELibraryStorageItemDownloadStatus Enum=NotDownloaded
   ) {is=title K2Node_FunctionEntry}
   n2 = Set Memory Format!(NumBytes:Integer64=@n1.DownloadSpaceSize) -> (FormattedMemory:Text) {is=title K2Node_MacroInstance}
   n3 = Format Text!(
    Format="LOCTEXT(\"{DownloadProgress}% of {DownloadSpaceSize}\")"
    DownloadProgress:Text="To Text (Integer)"(Value=@n1.DownloadProgress)
    DownloadSpaceSize:Text=@n2
   ) -> (Result)
   n4 = Set Memory Format(NumBytes=@n1)
   > Return Node(Text=Select(
     NotDownloaded=@n4
     Downloaded=@n4
     Downloading=@n3
     UpdateRequired=@n2
     Updating=@n3
     Queued=@n3
     Paused=@n3
     Removing=@n4
     Index=@n1.DownloadStatus
    ))
  ]}
 OnButtonSelected_Touch {nodes [OnButtonSelected_Touch!() -> (then, ListEntry:Common User Widget Object Reference) {is=title K2Node_FunctionEntry}]}
 OnMouseButtonDown {nodes [
   n1 = OnMouseButtonDown!() -> (then, MyGeometry:Geometry Structure, MouseEvent:"Pointer Event Structure (by ref)") {is=title K2Node_FunctionEntry}
   > n2 = Branch(Condition=Is Mouse Button Down!(Input:"Pointer Event Structure (by ref)"=@n1.MouseEvent, MouseButton:Key Structure=RightMouseButton) -> (ReturnValue:bool))
   > ToggleOpen(self=Get Anchor(self=Get ContextMenu()), bFocusOnOpen=false)
   > Return Node!(execute, ReturnValue:Event Reply Structure=Handled())
   Return Node(execute=@n2.else, ReturnValue=Unhandled())
  ]}
 SetContentContextMenu {nodes [
   SetContentContextMenu() {is=title K2Node_FunctionEntry}
   > n1 = Sequence!(execute) -> (then_0, then_1, then_2:Exec, then_3:Exec, then_4:Exec, then_5:Exec, then_6:Exec)
   n2 = Get DownloadStatus(self=Get FortLibraryStorageItemViewModel()) -> (DownloadStatus=NotDownloaded)
   n3 = Branch(
    execute=@n1.then_4
    Condition=Select(
     NotDownloaded:bool=false
     Downloaded:bool=false
     Downloading:bool=true
     UpdateRequired:bool=false
     Updating:bool=true
     Queued:bool=true
     Paused:bool=true
     PendingRemoval:bool=false
     Removing:bool=false
     Index=@n2
    ) -> (ReturnValue:bool)
   )
   n4 = Get ContextMenuActionContext() -> (ContextMenuActionContext:UIKit Context Menu Action Context Object Reference)
   > n5 = Branch(Condition=FindInteractionByID!(self:UIKit Context Menu Action Context Object Reference=@n4, InteractionNameID:Name=CancelDownload) -> (OutActionInfo:UIKit Context Menu Action Info Structure, ReturnValue:bool).ReturnValue)
   AddInteraction!(
    execute=@n5.else
    self:UIKit Context Menu Action Context Object Reference=@n4
    InActionInfo_DisplayName:Text="LOCTEXT(\"Cancel download\")"
    InActionInfo_IconTexture:Texture 2D Soft Object Reference
    InActionInfo_IconMaterial:Material Instance Soft Object Reference
    InActionInfo_SortPriority:int=250
    InActionInfo_bIsAvailable:bool=True
    InActionInfo_bIsEnabled:bool=True
    InActionInfo_bIsSelected:bool=False
    InActionInfo_bAddDivider:bool=False
    InActionInfo_bPositionDividerAfter:bool=False
    InActionInfo_bCloseMenuOnSelect:bool=False
    InActionInfo_InteractionNameID:Name=CancelDownload
    InActionInfo_Interaction:Data Table Row Handle Structure=(DataTable=None,RowName="")
    InActionInfo_PressedDelegate:Delegate="(null).None"
    InActionInfo_ReleasedDelegate:Delegate="(null).None"
    InInteractionDelegate:"Delegate (by ref)"=Create Event()
   ) -> (then)
   RemoveInteraction!(
    execute=@n3.else
    self:UIKit Context Menu Action Context Object Reference=Get ContextMenuActionContext()
    InteractionNameID:Name=CancelDownload
   ) -> (then)
   n6 = Branch(
    execute=@n1.then_1
    Condition=Select(
     NotDownloaded=false
     Downloaded=false
     Downloading=false
     UpdateRequired=true
     Updating=false
     Queued=false
     Paused=false
     PendingRemoval=false
     Removing=false
     Index=@n2
    )
   )
   n7 = Get ContextMenuActionContext()
   > n8 = Branch(Condition=FindInteractionByID(self=@n7, InteractionNameID=UpdateNow).ReturnValue)
   AddInteraction(
    execute=@n8.else
    self=@n7
    InActionInfo_DisplayName="LOCTEXT(\"Update now\")"
    InActionInfo_SortPriority=100
    InActionInfo_bIsAvailable=True
    InActionInfo_bIsEnabled=True
    InActionInfo_bIsSelected=False
    InActionInfo_bAddDivider=False
    InActionInfo_bPositionDividerAfter=False
    InActionInfo_bCloseMenuOnSelect=False
    InActionInfo_InteractionNameID=UpdateNow
    InActionInfo_Interaction=(DataTable=None,RowName="")
    InActionInfo_PressedDelegate="(null).None"
    InActionInfo_ReleasedDelegate="(null).None"
    InInteractionDelegate=Create Event()
   )
   RemoveInteraction(execute=@n6.else, self=Get ContextMenuActionContext(), InteractionNameID=UpdateNow)
   n9 = Branch(
    execute=@n1.then_3
    Condition=Select(
     NotDownloaded=false
     Downloaded=false
     Downloading=true
     UpdateRequired=false
     Updating=true
     Queued=false
     Paused=false
     PendingRemoval=false
     Removing=false
     Index=@n2
    )
   )
   n10 = Get ContextMenuActionContext()
   > n11 = Branch(Condition=FindInteractionByID(self=@n10, InteractionNameID=PauseDownload).ReturnValue)
   AddInteraction(
    execute=@n11.else
    self=@n10
    InActionInfo_DisplayName="LOCTEXT(\"Pause download\")"
    InActionInfo_SortPriority=200
    InActionInfo_bIsAvailable=True
    InActionInfo_bIsEnabled=True
    InActionInfo_bIsSelected=False
    InActionInfo_bAddDivider=False
    InActionInfo_bPositionDividerAfter=False
    InActionInfo_bCloseMenuOnSelect=False
    InActionInfo_InteractionNameID=PauseDownload
    InActionInfo_Interaction=(DataTable=None,RowName="")
    InActionInfo_PressedDelegate="(null).None"
    InActionInfo_ReleasedDelegate="(null).None"
    InInteractionDelegate=Create Event()
   )
   RemoveInteraction(execute=@n9.else, self=Get ContextMenuActionContext(), InteractionNameID=PauseDownload)
   n12 = Branch(
    execute=@n1.then_2
    Condition=Select(
     NotDownloaded=false
     Downloaded=false
     Downloading=false
     UpdateRequired=false
     Updating=false
     Queued=false
     Paused=true
     PendingRemoval=false
     Removing=false
     Index=@n2
    )
   )
   n13 = Get ContextMenuActionContext()
   > n14 = Branch(Condition=FindInteractionByID(self=@n13, InteractionNameID=ResumeDownload).ReturnValue)
   AddInteraction(
    execute=@n14.else
    self=@n13
    InActionInfo_DisplayName="LOCTEXT(\"Resume download\")"
    InActionInfo_SortPriority=150
    InActionInfo_bIsAvailable=True
    InActionInfo_bIsEnabled=True
    InActionInfo_bIsSelected=False
    InActionInfo_bAddDivider=False
    InActionInfo_bPositionDividerAfter=False
    InActionInfo_bCloseMenuOnSelect=False
    InActionInfo_InteractionNameID=ResumeDownload
    InActionInfo_Interaction=(DataTable=None,RowName="")
    InActionInfo_PressedDelegate="(null).None"
    InActionInfo_ReleasedDelegate="(null).None"
    InInteractionDelegate=Create Event()
   )
   RemoveInteraction(execute=@n12.else, self=Get ContextMenuActionContext(), InteractionNameID=ResumeDownload)
   n15 = IsLibraryContentRemovalEnabled!(execute=@n1.then_5) -> (then, ReturnValue:bool)
   > n16 = Branch(Condition=AND Boolean(
     A=Select(
      NotDownloaded=false
      Downloaded=true
      Downloading=false
      UpdateRequired=true
      Updating=false
      Queued=false
      Paused=false
      PendingRemoval=false
      Removing=false
      Index=@n2
     )
     B=@n15
    ))
   n17 = Get ContextMenuActionContext()
   > n18 = Branch(Condition=FindInteractionByID(self=@n17, InteractionNameID=RemoveContent).ReturnValue)
   AddInteraction(
    execute=@n18.else
    self=@n17
    InActionInfo_DisplayName="LOCTEXT(\"Remove content\")"
    InActionInfo_SortPriority=300
    InActionInfo_bIsAvailable=True
    InActionInfo_bIsEnabled=True
    InActionInfo_bIsSelected=False
    InActionInfo_bAddDivider=False
    InActionInfo_bPositionDividerAfter=False
    InActionInfo_bCloseMenuOnSelect=False
    InActionInfo_InteractionNameID=RemoveContent
    InActionInfo_Interaction=(DataTable=None,RowName="")
    InActionInfo_PressedDelegate="(null).None"
    InActionInfo_ReleasedDelegate="(null).None"
    InInteractionDelegate=Create Event()
   )
   RemoveInteraction(execute=@n16.else, self=Get ContextMenuActionContext(), InteractionNameID=RemoveContent)
   n19 = Get ContextMenuActionContext()
   SetVisibility(
    execute=@n1.then_6
    self=Get ContextMenu()
    InVisibility=Select!(
     Option 0:ESlateVisibility Enum=Collapsed
     Option 1:ESlateVisibility Enum=SelfHitTestInvisible
     Index:bool=HasAvailableInteractions!(self:UIKit Context Menu Action Context Object Reference=@n19) -> (ReturnValue:bool)
    ) -> (ReturnValue:ESlateVisibility Enum)
   )
   n20 = Branch(
    execute=@n1.then_0
    Condition=Select!(
     NotDownloaded:bool=false
     Downloaded:bool=false
     Downloading:bool=false
     UpdateRequired:bool=false
     Updating:bool=false
     Queued:bool=true
     Paused:bool=false
     PendingRemoval:bool=false
     Removing:bool=false
     Index:ELibraryStorageItemDownloadStatus Enum=@n2
    ) -> (ReturnValue:bool)
   )
   n21 = Get ContextMenuActionContext()
   > n22 = Branch(Condition=FindInteractionByID(self=@n21, InteractionNameID=DownloadNow).ReturnValue)
   AddInteraction(
    execute=@n22.else
    self=@n21
    InActionInfo_DisplayName="LOCTEXT(\"Download now\")"
    InActionInfo_SortPriority=50
    InActionInfo_bIsAvailable=True
    InActionInfo_bIsEnabled=True
    InActionInfo_bIsSelected=False
    InActionInfo_bAddDivider=False
    InActionInfo_bPositionDividerAfter=False
    InActionInfo_bCloseMenuOnSelect=False
    InActionInfo_InteractionNameID=DownloadNow
    InActionInfo_Interaction=(DataTable=None,RowName="")
    InActionInfo_PressedDelegate="(null).None"
    InActionInfo_ReleasedDelegate="(null).None"
    InInteractionDelegate=Create Event()
   )
   RemoveInteraction(execute=@n20.else, self=Get ContextMenuActionContext(), InteractionNameID=DownloadNow)
  ]}
 SetDownloadProgressValue {nodes [
   n1 = SetDownloadProgressValue!() -> (then, DownloadProgressValue:double=~) {is=title K2Node_FunctionEntry}
   > SetScalarParameterValue(self=Get ThumbnailMaterial() -> (ThumbnailMaterial:Material Instance Dynamic Object Reference), ParameterName=DownloadProgress, Value=@n1)
  ]}
 SetDownloadStatus {nodes [
   n1 = SetDownloadStatus!() -> (then, DownloadStatus:ELibraryStorageItemDownloadStatus Enum=NotDownloaded) {is=title K2Node_FunctionEntry}
   > SetContentContextMenu() {is=title K2Node_CallFunction}
   n2 = "Equal (Enum)"(A=@n1, B=PendingRemoval)
   n3 = Get FortLibraryStorageItemViewModel()
   n4 = Get TXT_Description() -> (TXT_Description:Common Text Object Reference)
   > Set Text(
    Text=SelectText(
     A=Get DescriptionText_PendingRemoval() -> (DescriptionText_PendingRemoval:Text)
     B=Get Description!(self:Fort Library Storage Item View Model Object Reference=@n3) -> (Description:Text)
     bPickA=@n2
    )
    self=@n4
   )
   > SetColorAndOpacity(
    self=@n4
    InColorAndOpacity=Select!(
     Option 0:Slate Color Structure=Get DescriptionColor_Base() -> (DescriptionColor_Base:Slate Color Structure)
     Option 1:Slate Color Structure=Get DescriptionColor_PendingRemoval() -> (DescriptionColor_PendingRemoval:Slate Color Structure)
     Index:bool=@n2
    ) -> (ReturnValue:Slate Color Structure)
   )
  ]}
 SetMemoryFormat {nodes [Outputs!(FormattedMemory:Text=AsMemory!(
     NumBytes:Integer64=Inputs!() -> (NumBytes:Integer64)
     UnitStandard:EMemoryUnitStandard Enum=SI
     bUseGrouping:bool=false
     MinimumIntegralDigits:int=1
     MaximumIntegralDigits:int=324
     MinimumFractionalDigits:int=0
     MaximumFractionalDigits:int=2
    ) -> (ReturnValue:Text))]}
 SetRemovingValue {nodes [
   n1 = SetRemovingValue!() -> (then, RemovingValue:double=~) {is=title K2Node_FunctionEntry}
   > SetScalarParameterValue(self=Get ThumbnailMaterial(), ParameterName=IsRemoving, Value=@n1)
  ]}
 SetTextureThumbnail {nodes [
   n1 = SetTextureThumbnail!() -> (then, Icon:ELibraryStorageItemIconType Enum=Core) {is=title K2Node_FunctionEntry}
   > SetTextureParameterValue(
    self=Get ThumbnailMaterial()
    ParameterName=Thumbnail
    Value=Select!(
     Core:Texture 2D Object Reference=/DiscoveryBrowser/TabScreens/Library/Textures/T_UI_Library_Storage_CoreContent
     BattleRoyale:Texture 2D Object Reference=/DiscoveryBrowser/TabScreens/Library/Textures/T_UI_ThinClient_Package_BR
     Juno:Texture 2D Object Reference=/DiscoveryBrowser/TabScreens/Library/Textures/T_UI_ThinClient_Package_Lego
     Sparks:Texture 2D Object Reference=/DiscoveryBrowser/TabScreens/Library/Textures/T_UI_ThinClient_Package_Festival
     DelMar:Texture 2D Object Reference=/DiscoveryBrowser/TabScreens/Library/Textures/T_UI_ThinClient_Package_RocketRacing
     Creative:Texture 2D Object Reference=/DiscoveryBrowser/TabScreens/Library/Textures/T_UI_ThinClient_Package_Creative
     HighResTexture:Texture 2D Object Reference=/DiscoveryBrowser/TabScreens/Library/Textures/T_UI_ThinClient_Package_HighRes
     Index:ELibraryStorageItemIconType Enum=@n1
    ) -> (ReturnValue:Texture 2D Object Reference)
   )
  ]}
 SetupContextMenu {nodes [
   SetupContextMenu() {is=title K2Node_FunctionEntry}
   > n1 = Construct Object from Class(Class=/Script/FortniteUI.UIKitContextMenuActionContext) -> (ReturnValue:UIKit Context Menu Action Context Object Reference)
   > Set ContextMenuActionContext(ContextMenuActionContext:UIKit Context Menu Action Context Object Reference=@n1) -> (Output_Get:UIKit Context Menu Action Context Object Reference)
   > SetContext!(
    execute
    self:WBP UIKit Context Menu Anchor Button Base Object Reference=Get ContextMenu()
    InteractionContext_Data:UIKit Context Menu Action Context Object Reference=Get ContextMenuActionContext()
   ) -> (then)
  ]}
 UpdateBottomBarButtonsVisibility {nodes [
   n1 = UpdateBottomBarButtonsVisibility!() -> (then, NewInputType:ECommonInputType Enum, bIsHovered:bool=~) {is=title K2Node_FunctionEntry}
   n2 = Get ContextMenu()
   n3 = Select!(MouseAndKeyboard:bool=false, Gamepad:bool=true, Touch:bool=false, Count:bool=false, Index:ECommonInputType Enum=@n1) -> (ReturnValue:bool)
   > SetBottomButtonVisibility!(
    execute
    self:WBP Bottom Bar Decoy Button Object Reference=Get WBP_BottomBarHackButton_ContextMenu() -> (WBP_BottomBarHackButton_ContextMenu:WBP Bottom Bar Decoy Button Object Reference)
    ShouldButtonBeVisible:bool=AND Boolean!(A=@n1.bIsHovered, B=@n3, C:bool=IsVisible(self=@n2)) -> (ReturnValue)
   ) -> (then)
   > Set UseBottomBarButtons!(execute, UseBottomBarButtons:bool=@n3, self:WBP UIKit Context Menu Anchor Button Base Object Reference=Get ContextMenu()) -> (then, Output_Get:bool)
  ]}
 UpdateSelectionState_Touch {nodes [
   UpdateSelectionState_Touch!() -> (then, bIsSelected:bool=~) {is=title K2Node_FunctionEntry}
   > RequestTransitionOnBlocks!(
    execute
    self:UIKit Modular Button Object Reference=Get WBP_Discover_Button_Generic()
    State:EUIKitBlockVisualState Enum=Select!(
     Option 0:EUIKitBlockVisualState Enum=Unfocused
     Option 1:EUIKitBlockVisualState Enum=Focused
     Index:bool=Get bIsSelected() -> (bIsSelected:bool)
    ) -> (ReturnValue:EUIKitBlockVisualState Enum)
   ) -> (then)
  ]}
]
variables [
 ContextMenuActionContext {type=UIKitContextMenuActionContext, default=null}
 DescriptionColor_Base {type=SlateColor, default {colorUseRule=UseColor_Specified, specifiedColor=<rgba 0.5776 0.5776 0.5776 1>}}
 DescriptionColor_PendingRemoval {type=SlateColor, default {colorUseRule=UseColor_Specified, specifiedColor=<rgba 1 0.147 0.1329 1>}}
 DescriptionText_PendingRemoval {type=string, default=This content will be removed next time you start Fortnite.}
 OnButtonSelected_Touch {}
 ThumbnailMaterial {type=MaterialInstanceDynamic}
]
widgets {
 info {
  inheritedWidgetCount=0
  namedSlotCount=1
  parentClass=&CommonUserWidget
  rootWidgetClass=&VerticalBox
  widgetCount=19
 }
 widgets [
  VB_Content {widgetClassPath=&VerticalBox, widget=VB_Content}
  IMG_DividerCore {
   widgetClassPath=&Image
   bIsVariable=true
   parent=VB_Content
   widget=IMG_DividerCore
   properties {
    visibility=Collapsed
    brush {
     bIsDynamicallyLoaded=false
     drawAs=Image
     imageType=FullColor
     mirroring=NoMirror
     resourceName=None
     resourceObject=/FortUILibrary/UIKit/Misc/MaterialInstances/MI_UIKit_HorizontalDividerLine
     tiling=NoTile
     imageSize=<1 2>
     margin=<ltrb 0 0 0 0>
     outlineSettings {
      bUseBrushTransparency=false
      roundingType=HalfHeightRadius
      width=0
      color {colorUseRule=UseColor_Specified, specifiedColor=<rgba 0 0 0 0>}
      cornerRadii=<0 0 0 0>
     }
     tintColor {colorUseRule=UseColor_Specified, specifiedColor=<rgba 1 1 1 1>}
     uVRegion {bIsValid=false, max=<0 0>, min=<0 0>}
    }
   }
   slot {class=VerticalBoxSlot, properties {verticalAlignment=VAlign_Top, padding=<ltrb 0 0 0 16>}}
  }
  WBP_Discover_Button_Generic {
   widgetClassPath=/DiscoveryBrowser/Tiles/WBP_Discover_Button_Generic
   bIsVariable=true
   parent=VB_Content
   widget=WBP_Discover_Button_Generic
   properties {
    backgroundMaterial=/DiscoveryBrowser/TabScreens/Library/MaterialInstances/MI_UI_Library_StorageItem_Bg
    minHeight=96
    minWidth=1088
    mobileMinHeight=96
    mobileMinWidth=1088
    outlineMaterial=/FortUILibrary/UIKit/Buttons/MaterialInstances/CTAButtons/MI_UIKit_ButtonOutline
    useTextBlock=false
   }
   slot {class=VerticalBoxSlot}
  }
  O_ContentContainer {widgetClassPath=&Overlay, namedSlotHost=WBP_Discover_Button_Generic, widget=O_ContentContainer}
  HB_Content {
   widgetClassPath=&HorizontalBox
   parent=O_ContentContainer
   widget=HB_Content
   slot {class=OverlaySlot, properties {horizontalAlignment=HAlign_Fill, verticalAlignment=VAlign_Fill, padding=<ltrb 16 16 32 16>}}
  }
  O_Thumbnail {
   widgetClassPath=&Overlay
   parent=HB_Content
   widget=O_Thumbnail
   slot {class=HorizontalBoxSlot, properties {horizontalAlignment=HAlign_Left, verticalAlignment=VAlign_Center, padding=<ltrb 0 0 16 0>}}
  }
  Img_Thumbnail {
   widgetClassPath=&Image
   bIsVariable=true
   parent=O_Thumbnail
   widget=Img_Thumbnail
   properties {
    visibility=SelfHitTestInvisible
    brush {
     bIsDynamicallyLoaded=false
     drawAs=Image
     imageType=FullColor
     mirroring=NoMirror
     resourceName=None
     resourceObject=/DiscoveryBrowser/TabScreens/Library/MaterialInstances/MI_UI_Library_Storage_Thumbnail
     tiling=NoTile
     imageSize=<64 64>
     margin=<ltrb 0 0 0 0>
     outlineSettings {
      bUseBrushTransparency=false
      roundingType=HalfHeightRadius
      width=0
      color {colorUseRule=UseColor_Specified, specifiedColor=<rgba 0 0 0 0>}
      cornerRadii=<0 0 0 0>
     }
     tintColor {colorUseRule=UseColor_Specified, specifiedColor=<rgba 1 1 1 1>}
     uVRegion {bIsValid=false, max=<0 0>, min=<0 0>}
    }
   }
   slot {class=OverlaySlot, properties {horizontalAlignment=HAlign_Fill, verticalAlignment=VAlign_Fill}}
  }
  StatusIndicator_Pause {
   widgetClassPath=/FortUILibrary/UIKit/StatusIndicator/WBP_UIKit_StatusIndicator
   bIsVariable=true
   parent=O_Thumbnail
   widget=StatusIndicator_Pause
   properties {type=Utility, overrideStatusIndicatorVisibility=true, visibility=Hidden}
   slot {class=OverlaySlot, properties {horizontalAlignment=HAlign_Center, verticalAlignment=VAlign_Center}}
  }
  VB_ContentInfos {
   widgetClassPath=&VerticalBox
   parent=HB_Content
   widget=VB_ContentInfos
   slot {class=HorizontalBoxSlot, properties {verticalAlignment=VAlign_Center, padding=<ltrb 0 0 16 0>, size {sizeRule=Fill, value=1}}}
  }
  TXT_ContentName {
   widgetClassPath=&CommonTextBlock
   bIsVariable=true
   parent=VB_ContentInfos
   widget=TXT_ContentName
   properties {
    mobileFontSizeMultiplier=1.1
    scrollStyle=&TextScrollStyle-Base_C
    text=Content Pack Name
    font {
     bForceMonospaced=false
     bMaterialIsStencil=false
     fontMaterial=null
     fontObject=&HeadingNow
     letterSpacing=0
     monospacedWidth=1
     size=15.75
     skewAmount=0
     typefaceFontName=74 Regular
     outlineSettings {
      bApplyOutlineToDropShadows=false
      bMiteredCorners=false
      bSeparateFillAlpha=false
      outlineMaterial=null
      outlineSize=0
      outlineColor=<rgba 0 0 0 1>
     }
    }
    shadowOffset=<0 0>
   }
   slot {class=VerticalBoxSlot, properties {verticalAlignment=VAlign_Top}}
  }
  TXT_Description {
   widgetClassPath=&CommonTextBlock
   bIsVariable=true
   parent=VB_ContentInfos
   widget=TXT_Description
   properties {
    mobileFontSizeMultiplier=1.25
    scrollStyle=&TextScrollStyle-Base_C
    text=Description
    colorAndOpacity {colorUseRule=UseColor_Specified, specifiedColor=<rgba 0.5776 0.5776 0.5776 1>}
    font {
     bForceMonospaced=false
     bMaterialIsStencil=false
     fontMaterial=null
     fontObject=&HeadingNow
     letterSpacing=0
     monospacedWidth=1
     size=11.25
     skewAmount=0
     typefaceFontName=74 Regular
     outlineSettings {
      bApplyOutlineToDropShadows=false
      bMiteredCorners=false
      bSeparateFillAlpha=false
      outlineMaterial=null
      outlineSize=0
      outlineColor=<rgba 0 0 0 1>
     }
    }
    shadowOffset=<0 0>
   }
   slot {class=VerticalBoxSlot, properties {verticalAlignment=VAlign_Top}}
  }
  SB_Status {
   widgetClassPath=&SizeBox
   parent=HB_Content
   widget=SB_Status
   properties {bOverride_WidthOverride=true, widthOverride=304}
   slot {class=HorizontalBoxSlot, properties {verticalAlignment=VAlign_Center, padding=<ltrb 0 0 16 0>}}
  }
  VB_Status {widgetClassPath=&VerticalBox, parent=SB_Status, widget=VB_Status, slot {class=SizeBoxSlot}}
  TXT_Status {
   widgetClassPath=&CommonTextBlock
   bIsVariable=true
   parent=VB_Status
   widget=TXT_Status
   properties {
    mobileFontSizeMultiplier=1.1
    scrollStyle=&TextScrollStyle-Base_C
    text=Downloading
    font {
     bForceMonospaced=false
     bMaterialIsStencil=false
     fontMaterial=null
     fontObject=&HeadingNow
     letterSpacing=0
     monospacedWidth=1
     size=15.75
     skewAmount=0
     typefaceFontName=74 Regular
     outlineSettings {
      bApplyOutlineToDropShadows=false
      bMiteredCorners=false
      bSeparateFillAlpha=false
      outlineMaterial=null
      outlineSize=0
      outlineColor=<rgba 0 0 0 1>
     }
    }
    shadowOffset=<0 0>
   }
   slot {class=VerticalBoxSlot, properties {verticalAlignment=VAlign_Top}}
  }
  TXT_SubStatus {
   widgetClassPath=&CommonTextBlock
   bIsVariable=true
   parent=VB_Status
   widget=TXT_SubStatus
   properties {
    mobileFontSizeMultiplier=1.25
    scrollStyle=&TextScrollStyle-Base_C
    text="{0}% of {0} GB"
    colorAndOpacity {colorUseRule=UseColor_Specified, specifiedColor=<rgba 0.5776 0.5776 0.5776 1>}
    font {
     bForceMonospaced=false
     bMaterialIsStencil=false
     fontMaterial=null
     fontObject=&HeadingNow
     letterSpacing=0
     monospacedWidth=1
     size=11.25
     skewAmount=0
     typefaceFontName=74 Regular
     outlineSettings {
      bApplyOutlineToDropShadows=false
      bMiteredCorners=false
      bSeparateFillAlpha=false
      outlineMaterial=null
      outlineSize=0
      outlineColor=<rgba 0 0 0 1>
     }
    }
    shadowOffset=<0 0>
   }
   slot {class=VerticalBoxSlot, properties {verticalAlignment=VAlign_Top}}
  }
  Switcher_InteractionCue {
   widgetClassPath=&CommonVisibilitySwitcher
   bIsVariable=true
   parent=HB_Content
   widget=Switcher_InteractionCue
   slot {class=HorizontalBoxSlot, properties {horizontalAlignment=HAlign_Right, verticalAlignment=VAlign_Center}}
  }
  ContextMenu {
   widgetClassPath=/FortUILibrary/UIKit/ContextMenu/WBP_UIKit_ContextMenuAnchorButton
   bIsVariable=true
   parent=Switcher_InteractionCue
   widget=ContextMenu
   properties {
    button Background Material=null
    button Outline Material=null
    buttonClass=&/WBP_Library_ContextMenu_Button.WBP_Library_ContextMenu_Button_C
    buttonIcon=/FortUILibrary/Textures/Icons/T_UI_Icon_UIKit_ContextMenu_SDF
    menuPlacement=MenuPlacement_CenteredBelowAnchor
    min Height=32
    min Width=32
    useBottomBarButtons=true
    button Padding=<ltrb 0 0 0 0>
   }
   slot {class=CommonVisibilitySwitcherSlot, properties {horizontalAlignment=HAlign_Center, verticalAlignment=VAlign_Center}}
  }
  Throbber {
   widgetClassPath=/FortUILibrary/UIKit/Throbber/WBP_UIKit_Throbber
   parent=Switcher_InteractionCue
   widget=Throbber
   slot {class=CommonVisibilitySwitcherSlot, properties {horizontalAlignment=HAlign_Center, verticalAlignment=VAlign_Center}}
  }
  WBP_BottomBarHackButton_ContextMenu {
   widgetClassPath=/Game/UI/Frontend/BottomBar/WBP_BottomBarDecoyButton
   bIsVariable=true
   parent=O_ContentContainer
   widget=WBP_BottomBarHackButton_ContextMenu
   properties {triggeringInputAction {dataTable=/FortUILibrary/Foundation/ActivityBrowserInputActionTable, rowName=ContextMenu}}
   slot {class=OverlaySlot, properties {horizontalAlignment=HAlign_Right, verticalAlignment=VAlign_Bottom}}
  }
 ]
}
