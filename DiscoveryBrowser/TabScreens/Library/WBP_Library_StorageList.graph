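# Storage list widget for the Library tab, showing Fortnite experience downloads as storage entries.
# On construct it initializes the storage view model if the feature is enabled; OnStorageItemChanged refills the list and, on gamepad, focuses an entry or fires OnNoNavigableWidgets.
# Touch selection is tracked through delegates bound per generated entry and reconciled when input method changes.
# Core content items are non-selectable, and a core first item leaves no navigable widget.
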WidgetBlueprint /DiscoveryBrowser/TabScreens/Library/WBP_Library_StorageList
graphs [
 EventGraph {nodes [
   n1 = "On Entry Generated (ListView)"!() -> (OutputDelegate, then, Widget:User Widget Object Reference) {is=title K2Node_ComponentBoundEvent}
   > n2 = Cast To WBP_Library_StorageItemEntry!(execute, Object=@n1.Widget) -> (then, CastFailed, AsWBP Library Storage Item Entry:WBP Library Storage Item Entry Object Reference) {is=title K2Node_DynamicCast}
   n3 = Create Event()
   > Bind Event to On Button Selected Touch(self:WBP Library Storage Item Entry Object Reference=@n2, Delegate=@n3) {is=title K2Node_AddDelegate}
   n4 = "On Entry Released (ListView)"!() -> (OutputDelegate, then, Widget:User Widget Object Reference) {is=title K2Node_ComponentBoundEvent}
   > n5 = Cast To WBP_Library_StorageItemEntry(Object=@n4.Widget)
   > Unbind Event from On Button Selected Touch(self:WBP Library Storage Item Entry Object Reference=@n5, Delegate=@n3) {is=title K2Node_RemoveDelegate}
   n6 = HandleOnButtonSelected_Touch!() -> (OutputDelegate, then, ListEntry:Common User Widget Object Reference) {is=title K2Node_CustomEvent}
   > n7 = Cast To WBP_Library_StorageItemEntry(Object=@n6.ListEntry)
   n8 = Get CurrentSelectedButton_Touch() -> (CurrentSelectedButton_Touch:WBP Library Storage Item Entry Object Reference)
   > n9 = Is Valid(InputObject=@n8)
   > n10 = UpdateSelectionState_Touch!(execute, self:WBP Library Storage Item Entry Object Reference=@n8, bIsSelected:bool=false) -> (then)
   UpdateSelectionState_Touch(execute=[@n10.then, @n9.Is Not Valid], self=@n7, bIsSelected=true)
   > n11 = Set CurrentSelectedButton_Touch(CurrentSelectedButton_Touch:WBP Library Storage Item Entry Object Reference=@n7) -> (Output_Get:WBP Library Storage Item Entry Object Reference)
   > SetFocus(self=@n11)
   n12 = HandleOnInputMethodChanged!() -> (OutputDelegate, then, NewInputType:ECommonInputType Enum) {is=title K2Node_CustomEvent}
   > n13 = Switch on ECommonInputType!(execute, Selection:ECommonInputType Enum=@n12.NewInputType) -> (MouseAndKeyboard:Exec, Gamepad:Exec, Touch:Exec, Count:Exec) {is=title K2Node_SwitchEnum}
   Clear Selection(execute=@n13.Touch, self:List View Object Reference=Get ListView() -> (ListView:Common List View Object Reference))
   n14 = Get CurrentSelectedButton_Touch()
   Is Valid(exec=[@n13.Gamepad, @n13.MouseAndKeyboard], InputObject=@n14)
   > UpdateSelectionState_Touch(self=@n14, bIsSelected=false)
   > Set CurrentSelectedButton_Touch()
   Event Destruct()
   n15 = Create Event()
   n16 = Get CommonInputSubsystem()
   > Unbind Event from On Input Method Changed(self=@n16, Delegate=@n15)
   Event Construct()
   > Bind Event to On Input Method Changed(self=@n16, Delegate=@n15)
   > n17 = IsLibraryStorageEnabled!(execute) -> (then, ReturnValue:bool)
   > Branch(Condition=@n17)
   > InitializeProperties(self:Fort Library Storage View Model Object Reference=Get FortLibraryStorageViewModel() -> (FortLibraryStorageViewModel:Fort Library Storage View Model Object Reference))
  ]}
 GetItemSelectableOrNavigable {nodes [
   n1 = GetItemSelectableOrNavigable!() -> (then, Item:Object Reference) {is=title K2Node_FunctionEntry}
   > n2 = Cast To FortLibraryStorageItemViewModel!(execute, Object=@n1) -> (then, CastFailed, AsFort Library Storage Item View Model:Fort Library Storage Item View Model Object Reference) {is=title K2Node_DynamicCast}
   > Return Node!(
    execute
    ReturnValue:bool=NOT Boolean(A=Get bIsCoreContent!(self:Fort Library Storage Item View Model Object Reference=@n2) -> (bIsCoreContent:bool))
   )
  ]}
 GetNavigableWidget {nodes [
   GetNavigableWidget() {is=title K2Node_FunctionEntry}
   > n1 = Cast To FortLibraryStorageItemViewModel(Object=GetItemAt!(self:List View Object Reference=Get ListView(), Index:int=0) -> (ReturnValue:Object Reference))
   > Return Node!(
    execute
    NavigableWidget:Widget Object Reference=Select(Option 0:Widget Object Reference=Get ListView(), Option 1:Widget Object Reference, Index=Get bIsCoreContent(self=@n1)) -> (ReturnValue:Widget Object Reference)
   )
  ]}
 NavigateToTop {nodes [
   NavigateToTop() {is=title K2Node_FunctionEntry}
   n1 = Get ListView()
   > NavigateToIndex!(execute, self:List View Object Reference=@n1, Index:int=0) -> (then)
   > SetSelectedIndex!(execute, self:List View Object Reference=@n1, Index:int=0) -> (then)
   > SetUserFocus!(execute, self:Widget Object Reference=@n1, PlayerController:Player Controller Object Reference=GetOwningPlayer()) -> (then)
  ]}
 OnNoNavigableWidgets {nodes [OnNoNavigableWidgets() {is=title K2Node_FunctionEntry}]}
 OnStorageItemChanged {nodes [
   n1 = OnStorageItemChanged!() -> (then, ListItems:Array of Fort Library Storage Item View Model Object References) {is=title K2Node_FunctionEntry}
   n2 = Get ListView()
   > Set List Items!(execute, self:List View Object Reference=@n2, InListItems:Array of Object References=@n1) -> (then) {is=title K2Node_CallFunction}
   n3 = Get CommonInputSubsystem()
   > Branch(Condition="Equal (Enum)"(A=GetCurrentInputType(self=@n3), B=Gamepad))
   > n4 = Branch(Condition="integer <= integer"(A=GetNumItems!(self:List View Object Reference=@n2) -> (ReturnValue:int=0), B=1))
   > Call On No Navigable Widgets!(execute:Exec, self:WBP Library Storage List Object Reference) -> (then:Exec) {is=title K2Node_CallDelegate}
   SetFocus(
    execute=@n4.else
    self="Get (a copy)"(Array:Array of User Widget Object References=GetDisplayedEntryWidgets!(self:List View Base Object Reference=@n2) -> (ReturnValue:Array of User Widget Object References)) -> (Output:User Widget Object Reference)
   )
   > SetSelectedIndex(self=@n2, Index=0)
  ]}
]
variables [
 CurrentSelectedButton_Touch {type=WBP_Library_StorageItemEntry_C, default=null}
 OnNoNavigableWidgets {}
]
widgets {
 info {inheritedWidgetCount=1, parentClass=/Script/FortniteUI.FortPolymorphicListView, rootWidgetClass=&Overlay, widgetCount=6}
 widgets [
  Overlay_0 {widgetClassPath=&Overlay, widget=Overlay_0}
  VB_Content {widgetClassPath=&VerticalBox, parent=Overlay_0, widget=VB_Content, slot {class=OverlaySlot}}
  TXT_Title {
   widgetClassPath=&CommonTextBlock
   bIsVariable=true
   parent=VB_Content
   widget=TXT_Title
   properties {
    autoWrapText=true
    lineHeightPercentage=1.15
    text=Storage
    colorAndOpacity {colorUseRule=UseColor_Specified, specifiedColor=<rgba 1 1 1 100>}
    font {
     bForceMonospaced=false
     bMaterialIsStencil=false
     fontMaterial=null
     fontObject=&HeadingNow
     letterSpacing=0
     monospacedWidth=1
     size=18
     skewAmount=0
     typefaceFontName=75 Medium
     outlineSettings {
      bApplyOutlineToDropShadows=false
      bMiteredCorners=false
      bSeparateFillAlpha=false
      outlineMaterial=null
      outlineSize=0
      outlineColor=<rgba 0 0 0 1>
     }
    }
    shadowOffset=<0 0>
   }
   slot {class=VerticalBoxSlot, properties {verticalAlignment=VAlign_Top}}
  }
  TXT_Subtitle {
   widgetClassPath=&CommonTextBlock
   bIsVariable=true
   parent=VB_Content
   widget=TXT_Subtitle
   properties {
    lineHeightPercentage=1.15
    mobileFontSizeMultiplier=1.25
    text=Manage downloads for Fortnite experiences
    colorAndOpacity {colorUseRule=UseColor_Specified, specifiedColor=<rgba 0.5776 0.5776 0.5776 1>}
    font {
     bForceMonospaced=false
     bMaterialIsStencil=false
     fontMaterial=null
     fontObject=&HeadingNow
     letterSpacing=0
     monospacedWidth=1
     size=11.25
     skewAmount=0
     typefaceFontName=74 Regular
     outlineSettings {
      bApplyOutlineToDropShadows=false
      bMiteredCorners=false
      bSeparateFillAlpha=false
      outlineMaterial=null
      outlineSize=0
      outlineColor=<rgba 0 0 0 1>
     }
    }
    shadowOffset=<0 0>
   }
   slot {class=VerticalBoxSlot, properties {verticalAlignment=VAlign_Top, padding=<ltrb 0 0 0 16>}}
  }
  Overlay_1 {
   widgetClassPath=&Overlay
   parent=Overlay_0
   widget=Overlay_1
   properties {clipping=ClipToBounds}
   slot {class=OverlaySlot, properties {horizontalAlignment=HAlign_Fill, verticalAlignment=VAlign_Fill, padding=<ltrb 0 64 0 0>}}
  }
  ListView {
   widgetClassPath=&CommonListView
   bInherited=true
   bIsVariable=true
   parent=Overlay_1
   widget=ListView
   properties {
    clipping=Inherit
    entryWidgetClass=&/WBP_Library_StorageItemEntry.WBP_Library_StorageItemEntry_C
    numDesignerPreviewEntries=10
    verticalEntrySpacing=16
    scrollBarPadding=<ltrb 16 0 0 0>
   }
   slot {class=OverlaySlot, properties {padding=<ltrb 16 16 16 16>}}
  }
 ]
}
