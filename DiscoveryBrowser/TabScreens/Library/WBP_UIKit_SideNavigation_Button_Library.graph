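# Library tab side navigation button in the Discovery Browser, a child of the UIKit side navigation button with no widgets of its own.
# Construct and PreConstruct only call their parent versions.
# SetTabIcon takes a reference link item view model and only prints its small image URL and texture name as debug output, without applying an icon.
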WidgetBlueprint /DiscoveryBrowser/TabScreens/Library/WBP_UIKit_SideNavigation_Button_Library
graphs [
 EventGraph {nodes [
   n1 = Event PreConstruct()
   > Parent: PreConstruct(IsDesignTime=@n1.IsDesignTime)
   Event Construct()
   > Parent: Construct()
   Event Tick()
  ]}
 SetTabIcon {nodes [
   n1 = SetTabIcon!() -> (then, ReferenceLink:Fort Discover Surface Tile Item VM Object Reference) {is=title K2Node_FunctionEntry}
   n2 = GetImageTexture!(self:Fort Discover Link Entry Tile Item VM Object Reference=@n1) -> (ReturnValue:Texture Object Reference)
   > PrintString!(
    execute
    InString=Append!(
     A="Image Icon: "
     B=GetImageURL!(self:Fort Discover Link Entry Tile Item VM Object Reference=@n1, InSize:ECreativeLinkPreviewSize Enum=Small) -> (ReturnValue:String)
     C:String=", Image Texture: "
     D:String=GetDisplayName(Object=@n2)
    ) -> (ReturnValue)
    bPrintToScreen
    bPrintToLog
    TextColor
    Duration
    Key
   ) -> (then)
  ]}
]
variables []
widgets {
 info {
  inheritedWidgetCount=0
  parentClass=/FortUILibrary/UIKit/Buttons/TabButtons/WBP_UIKit_SideNavigation_Button.WBP_UIKit_SideNavigation_Button_C
  rootWidgetClass=null
  widgetCount=0
 }
 widgets []
}
