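# Reusable activatable Discovery browser screen template pairing nested side navigation with one continuous panel list.
# A recursive walk of the panel tree view model fills both and records an index-to-view-model array, syncing tab clicks, list selection and hover.
# Switchers show loading, error, no-results or alternate content, with gamepad-style focus routing and touch swipes between rows.
# Class overrides and named slots customize buttons, divider, spinner and content.
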WidgetBlueprint /DiscoveryBrowser/WBP_Template_NestedSideNav_AndContinuousList
graphs [
 BP_GetDesiredFocusTarget {nodes [
   Get Desired Focus Target()
   > n1 = Branch(Condition=Get bIsCurrentTabAlternateContent() -> (bIsCurrentTabAlternateContent:bool))
   > n2 = GetAlternateContentNavigableWidget!(execute, self) -> (then, NavigableWidget:Widget Object Reference)
   > Return Node!(execute, ReturnValue:Widget Object Reference=@n2)
   n3 = Get ListView!(self:Fort Polymorphic List View Object Reference=Get WBP_Poly_MegaList() -> (WBP_Poly_MegaList:WBP Poly Mega List Object Reference)) -> (ListView:Common List View Object Reference)
   n4 = Branch(execute=@n1.else, Condition="integer <= integer"(A=GetNumItems!(self:List View Object Reference=@n3) -> (ReturnValue:int=0)))
   n5 = Get WBP_UIKit_SideNavigation() -> (WBP_UIKit_SideNavigation:WBP UIKit Side Navigation Object Reference)
   > Return Node(ReturnValue=GetTabButtonBaseByID!(
     self:Common Tab List Widget Base Object Reference=@n5
     TabNameID:Name=GetActiveTab!(self:Common Tab List Widget Base Object Reference=@n5) -> (ReturnValue:Name=None)
    ) -> (ReturnValue:Common Button Base Object Reference))
   n6 = GetIndexForItem!(
    self:List View Object Reference=@n3
    Item:Object Reference=Get Selected Item!(self:List View Object Reference=@n3) -> (ReturnValue:Object Reference) {is=title K2Node_CallFunction}
   ) -> (ReturnValue:int=0)
   SetSelectedIndex!(
    execute=@n4.else
    self:List View Object Reference=@n3
    Index:int=Select(Option 0:int=0, Option 1:int=@n6, Index="integer >= integer"(A=@n6)) -> (ReturnValue:int)
   ) -> (then)
   > Return Node(ReturnValue=@n3)
  ]}
 EventGraph {nodes [
   n1 = "On Tab Clicked (WBP_UIKit_SideNavigation)"!() -> (OutputDelegate, then, TabId:Name=None) {is=title K2Node_ComponentBoundEvent}
   > Call On Side Nav Outer Tab Clicked!(execute:Exec, self:WBP Template Nested Side Nav And Continuous List Object Reference, OuterTabID:Name=@n1.TabId) -> (then:Exec) {is=title K2Node_CallDelegate}
   n2 = Get WBP_UIKit_SideNavigation()
   n3 = GetInnerTabIdAtIndex!(self:WBP UIKit Side Navigation Base Object Reference=@n2, TabID:"Name (by ref)"=@n1.TabId, Index:int=0) -> (InnerTabID:Name=None)
   n4 = GetInnerTabsCount!(self:WBP UIKit Side Navigation Base Object Reference=@n2, TabID:Name=@n1.TabId) -> (Count:int=0)
   > SelectMegaListIndexThatMatchesViewModel!(
    execute
    self
    TabViewModel:MVVM Base Viewmodel Object Reference=Select(
     Option 0:MVVM Base Viewmodel Object Reference=GetSideNavigationTabViewModel!(self:WBP UIKit Side Navigation Base Object Reference=@n2, TabID:Name=@n1.TabId, InnerTabID:Name=None) -> (TabViewModel:MVVM Base Viewmodel Object Reference)
     Option 1:MVVM Base Viewmodel Object Reference=GetSideNavigationTabViewModel(self=@n2, TabID=@n1.TabId, InnerTabID=@n3)
     Index="integer > integer"(A=@n4)
    ) -> (ReturnValue:MVVM Base Viewmodel Object Reference)
   ) -> (then)
   n5 = "On Side Navigation Inner Tab Clicked (WBP_UIKit_SideNavigation)"!() -> (OutputDelegate, then, TabID:Name=None, InnerTabID:Name=None, TabViewModel:MVVM Base Viewmodel Object Reference) {is=title K2Node_ComponentBoundEvent}
   > SelectMegaListIndexThatMatchesViewModel(TabViewModel=@n5.TabViewModel)
   "On Swipe Up (SwipePanel)"() {is=title K2Node_ComponentBoundEvent}
   n6 = Get WBP_Poly_MegaList()
   > SwipeToPreviousRow_Touch(self:WBP Poly Mega List Object Reference=@n6)
   "On Swipe Down (SwipePanel)"() {is=title K2Node_ComponentBoundEvent}
   > SwipeToNextRow_Touch(self:WBP Poly Mega List Object Reference=@n6)
   "On Active Widget Index Changed BP (Switcher_LoadingState)"!() -> (OutputDelegate, then, ActiveWidget:Widget Object Reference, ActiveIndex:int=0) {is=title K2Node_ComponentBoundEvent}
   > RequestRefreshFocus(self:Common Activatable Widget Object Reference)
   n7 = OnMegaListSelectionChange!() -> (OutputDelegate, then, Item:Object Reference, bIsSelected:bool=~) {is=title K2Node_CustomEvent}
   > Branch(Condition=@n7.bIsSelected)
   n8 = Get ListView(self=Get WBP_Poly_MegaList())
   > SelectSideNavigationViewModel!(
    execute
    self:WBP UIKit Side Navigation Base Object Reference=Get WBP_UIKit_SideNavigation()
    TabViewModel:MVVM Base Viewmodel Object Reference="Get (a copy)"(
     Array:Array of MVVM Base Viewmodel Object References=Get MegaListToSideNavTabVMMapping() -> (MegaListToSideNavTabVMMapping:Array of MVVM Base Viewmodel Object References)
     Dimension 1=GetIndexForItem(self=@n8, Item=@n7.Item) -> (ReturnValue=0)
    ) -> (Output:MVVM Base Viewmodel Object Reference)
    bSuppressClickFeedback:bool=true
   ) -> (then, Success:bool)
   n9 = OnMegaListItemHoverChanged!() -> (OutputDelegate, then, Item:Object Reference, bIsHovered:bool=~) {is=title K2Node_CustomEvent}
   > Branch(Condition=@n9.bIsHovered)
   n10 = GetIndexForItem(self=Get ListView(self=Get WBP_Poly_MegaList()), Item=@n9.Item) -> (ReturnValue=0)
   > SelectSideNavigationViewModel(
    self=Get WBP_UIKit_SideNavigation()
    TabViewModel="Get (a copy)"(Array=Get MegaListToSideNavTabVMMapping(), Dimension 1=@n10)
    bSuppressClickFeedback=true
   )
   > SetSelectedIndex(self=Get ListView(self=Get WBP_Poly_MegaList()), Index=@n10)
   Event PreConstruct()
   > SetCompactUIStyle() {is=title K2Node_CallFunction}
   > SetTemplateOverrides()
   > n11 = GetDynamicMaterial(self=Get Image_HeaderSoftEdge() -> (Image_HeaderSoftEdge:Image Object Reference))
   > SetScalarParameterValue(self=@n11, ParameterName=Overlay Rotation, Value=0.5)
   n12 = Get Slot_CustomNoResults() -> (Slot_CustomNoResults:Named Slot Object Reference)
   > SetActiveWidget(
    self=Get EmptyList_NoResults() -> (EmptyList_NoResults:Common Visibility Switcher Object Reference)
    Widget=Select(
     Option 0:Widget Object Reference=Get NoResultsGenericFallback() -> (NoResultsGenericFallback:WBP Discovery Blank Tiles Rounded Corners Object Reference)
     Option 1:Widget Object Reference=@n12
     Index=HasAnyChildren!(self:Panel Widget Object Reference=@n12) -> (ReturnValue:bool)
    ) -> (ReturnValue:Widget Object Reference)
   )
   > n13 = Cast To CommonVisibilitySwitcherSlot!(
    execute
    Object=Get Slot!(self:Widget Object Reference=Get MegaListContainerOverlay() -> (MegaListContainerOverlay:Overlay Object Reference)) -> (Slot:Panel Slot Object Reference)
   ) -> (then, CastFailed, AsCommon Visibility Switcher Slot:Common Visibility Switcher Slot Object Reference) {is=title K2Node_DynamicCast}
   > Set Padding!(execute, Padding:Margin Structure=Get MegaListPadding() -> (MegaListPadding:Margin Structure), self:Overlay Slot Object Reference=@n13) -> (then, Output_Get:Margin Structure)
   Event Construct()
   n14 = Get ListView(self=Get WBP_Poly_MegaList())
   > Bind Event to On Item Selection Changed(self:List View Object Reference=@n14, Delegate=Create Event()) {is=title K2Node_AddDelegate}
   > Bind Event to On Item Is Hovered Changed(self:List View Object Reference=@n14, Delegate=Create Event()) {is=title K2Node_AddDelegate}
   Event Destruct()
   > Unbind all Events from On Item Selection Changed!(execute:Exec, self:List View Object Reference=@n14) -> (then:Exec) {is=title K2Node_ClearDelegate}
   > Unbind all Events from On Item Is Hovered Changed!(execute:Exec, self:List View Object Reference=@n14) -> (then:Exec) {is=title K2Node_ClearDelegate}
  ]}
 GetAlternateContentNavigableWidget {nodes [
   GetAlternateContentNavigableWidget() {is=title K2Node_FunctionEntry}
   n1 = Get AlternateContentNavigableWidget() -> (AlternateContentNavigableWidget:Widget Object Reference)
   > n2 = Is Valid(InputObject=@n1)
   n3 = Get WBP_UIKit_SideNavigation()
   Return Node!(
    execute=@n2.Is Not Valid
    NavigableWidget:Widget Object Reference=GetTabButtonBaseByID(self=Get WBP_UIKit_SideNavigation(), TabNameID=GetActiveTab(self=@n3) -> (ReturnValue=None))
   )
   > Return Node(NavigableWidget=@n1)
  ]}
 GiveFocusToSideNav {nodes [
   GiveFocusToSideNav() {is=title K2Node_FunctionEntry}
   n1 = Get WBP_UIKit_SideNavigation()
   > SetFocus(self=GetTabButtonBaseByID(self=@n1, TabNameID=GetActiveTab(self=@n1) -> (ReturnValue=None)))
  ]}
 On Navigate Right to List {nodes [
   On Navigate Right to List!() -> (then, Navigation:EUINavigation Enum) {is=title K2Node_FunctionEntry}
   > n1 = Branch(Condition=Get bIsCurrentTabAlternateContent())
   > n2 = GetAlternateContentNavigableWidget() {is=title K2Node_CallFunction}
   > Return Node!(execute, ReturnValue:Widget Object Reference=@n2)
   Return Node(execute=@n1.else, ReturnValue=Get ListView(self=Get WBP_Poly_MegaList()))
  ]}
 OnMegaListArrayFinishedRecursiveCreation {nodes [OnMegaListArrayFinishedRecursiveCreation!() -> (then, LastIndex:int) {is=title K2Node_FunctionEntry}]}
 OnNavigate Left To Tabs {nodes [
   OnNavigate Left To Tabs!() -> (then, Navigation:EUINavigation Enum) {is=title K2Node_FunctionEntry}
   n1 = Get WBP_UIKit_SideNavigation()
   n2 = GetActiveTab(self=@n1) -> (ReturnValue=None)
   > ClearSelection(self:Common Button Base Object Reference=GetTabButtonBaseByID(self=@n1, TabNameID=@n2))
   > Return Node(ReturnValue=GetInnerTabButton!(
     self:WBP UIKit Side Navigation Base Object Reference=@n1
     TabID:Name=@n2
     InnerTabID:Name=Get LastNotifiedInnerTabID!(self:WBP UIKit Side Navigation Base Object Reference=@n1) -> (LastNotifiedInnerTabID:Name=None)
    ) -> (TabButton:Common Button Base Object Reference))
  ]}
 OnSideNavOuterTabClicked {nodes [OnSideNavOuterTabClicked!() -> (then, OuterTabID:Name) {is=title K2Node_FunctionEntry}]}
 Recursive Tree Creation {nodes [
   n1 = Recursive Tree Creation!() -> (then, Node:Discover Panel Tree View Model Object Reference, IsIndented:bool=~) {is=title K2Node_FunctionEntry}
   n2 = Get Children!(self:Discover Panel Tree View Model Object Reference=@n1) -> (Children:Array of Discover Panel Tree View Model Object References)
   > n3 = Branch(Condition=Is Empty!(TargetArray:Array of Discover Panel Tree View Model Object References=@n2) -> (ReturnValue:bool))
   n4 = Get Panel!(self:Discover Panel Tree View Model Object Reference=@n1) -> (Panel:Discover Provider View Model Object Reference)
   > AddItem!(execute, self:List View Object Reference=Get ListView(self=Get WBP_Poly_MegaList()), Item:Object Reference=@n4) -> (then)
   > Add(
    TargetArray:Array of MVVM Base Viewmodel Object References=Get MegaListToSideNavTabVMMapping()
    NewItem:"MVVM Base Viewmodel Object Reference (by ref)"=@n1
   ) {is=title K2Node_CallArrayFunction}
   > n5 = Set AtLeastOneTileWasFound(AtLeastOneTileWasFound:bool=Is Not Empty!(TargetArray:Array of Discover Tile Item View Model Object References=Get Items!(self:Discover Provider View Model Object Reference=@n4) -> (Items:Array of Discover Tile Item View Model Object References)) -> (ReturnValue:bool)) -> (Output_Get:bool)
   Return Node!(execute=[@n5.then, @n6.Completed], OutAtLeastOneTileWasFound:bool=Get AtLeastOneTileWasFound() -> (AtLeastOneTileWasFound:bool))
   n6 = For Each Loop(Exec=@n3.else, Array:Array of Discover Panel Tree View Model Object References=@n2) -> (Array Element:Discover Panel Tree View Model Object Reference)
   > n7 = Branch(Condition=Get IsIndented() -> (IsIndented:bool))
   n8 = String To Name(InString=Get PanelName!(self:Discover Provider View Model Object Reference=Get Panel(self=@n6)) -> (PanelName:String))
   n9 = Get WBP_UIKit_SideNavigation()
   > n10 = AddSideNavigationInnerTab!(
    execute
    self:WBP UIKit Side Navigation Base Object Reference=@n9
    TabViewModel:MVVM Base Viewmodel Object Reference=@n6
    TabID:Name=String To Name(InString=Get PanelName(self=@n4))
    InnerTabID:Name=@n8
    Index:int=-1
   ) -> (then)
   n11 = Recursive Tree Creation!(execute=[@n12.then, @n10.then], self, Node:Discover Panel Tree View Model Object Reference=@n6, IsIndented:bool=true) -> (then, OutAtLeastOneTileWasFound:bool) {is=title K2Node_CallFunction}
   > Set AtLeastOneTileWasFound(AtLeastOneTileWasFound=OR Boolean(A=@n11, B=Get AtLeastOneTileWasFound()))
   n12 = AddSideNavigationTab!(
    execute=@n7.else
    self:WBP UIKit Side Navigation Base Object Reference=@n9
    TabViewModel:MVVM Base Viewmodel Object Reference=@n6
    TabID:Name=@n8
    Index:int=-1
   ) -> (then)
  ]}
 SelectMegaListIndexThatMatchesViewModel {nodes [
   n1 = SelectMegaListIndexThatMatchesViewModel!() -> (then, TabViewModel:MVVM Base Viewmodel Object Reference) {is=title K2Node_FunctionEntry}
   > n2 = For Each Loop with Break!(Exec:Exec, Array:Array of MVVM Base Viewmodel Object References=Get MegaListToSideNavTabVMMapping(), Break:Exec=@n4.then) -> (LoopBody:Exec, Array Element:MVVM Base Viewmodel Object Reference, Array Index:int, Completed:Exec)
   > Branch(Condition="Equal (Object)"(A=@n2, B=@n1))
   n3 = Get ListView(self=Get WBP_Poly_MegaList())
   > SetSelectedIndex(self=@n3, Index=@n2.Array Index)
   > n4 = SetScrollOffset!(execute, self:List View Base Object Reference=@n3, InScrollOffset:float="To Float (Integer)"(InInt=@n2.Array Index)) -> (then)
  ]}
 Set Content Padding {nodes [
   Set Content Padding() {is=title K2Node_FunctionEntry}
   > Set Padding(
    Padding=Get PaddingContentArea() -> (PaddingContentArea:Margin Structure)
    self:Grid Slot Object Reference=SlotAsGridSlot(Widget=Get SafeZone() -> (SafeZone:Safe Zone Object Reference))
   )
  ]}
 Set Custom Loading Spinner {nodes [
   Set Custom Loading Spinner() {is=title K2Node_FunctionEntry}
   > RemoveWidget!(execute, Widget:Widget Object Reference=Get LoadingSpinnerWidget() -> (LoadingSpinnerWidget:WBP UIKit Throbber Base Object Reference)) -> (then)
   n1 = Get LoadingSpinnerClass() -> (LoadingSpinnerClass:WBP UIKit Throbber Base Class Reference)
   > Branch(Condition=IsValidClass!(Class:Object Class Reference=@n1) -> (ReturnValue:bool))
   > n2 = Construct Object from Class(Class=@n1) -> (ReturnValue:WBP UIKit Throbber Base Object Reference)
   > n3 = Set LoadingSpinnerWidget(LoadingSpinnerWidget:WBP UIKit Throbber Base Object Reference=@n2) -> (Output_Get:WBP UIKit Throbber Base Object Reference)
   > Branch(Condition=IsValid(Object=@n3))
   > n4 = AddChildToOverlay!(
    execute
    self:Overlay Object Reference=Get Loading_SpinnerContainer() -> (Loading_SpinnerContainer:Overlay Object Reference)
    Content:Widget Object Reference=@n3
   ) -> (then, ReturnValue:Overlay Slot Object Reference)
   > SetHorizontalAlignment(self=@n4, InHorizontalAlignment=HAlign_Center)
   > SetVerticalAlignment(self=@n4, InVerticalAlignment=VAlign_Center)
   n5 = Get LoadingSpinnerWidget()
   > SetThrobberSize!(execute, self:WBP UIKit Throbber Base Object Reference=@n5, Size:E_UI_ThrobberSize Enum=NewEnumerator2) -> (then)
   > Set DisplayBackground!(execute, DisplayBackground:bool=false, self:WBP UIKit Throbber Base Object Reference=@n5) -> (then, Output_Get:bool)
  ]}
 Set Side Nav Custom Divider {nodes [
   Set Side Nav Custom Divider() {is=title K2Node_FunctionEntry}
   n1 = Get SideNav_DividerClass() -> (SideNav_DividerClass:WBP UIKit Divider Base Class Reference)
   > Branch(Condition=IsValidClass(Class=@n1))
   > Set DividerClass!(
    execute
    DividerClass:WBP UIKit Divider Base Class Reference=@n1
    self:WBP UIKit Tab Group Base Object Reference=Get WBP_UIKit_SideNavigation()
   ) -> (then, Output_Get:WBP UIKit Divider Base Class Reference)
  ]}
 Set Side Nav Custom Nested Button {nodes [
   Set Side Nav Custom Nested Button() {is=title K2Node_FunctionEntry}
   n1 = Get SideNav_Button_NestedClass() -> (SideNav_Button_NestedClass:WBP UIKit Button Generic Base Class Reference)
   > Branch(Condition=IsValidClass(Class=@n1))
   > Set InnerTabButtonClassOverride!(
    execute
    InnerTabButtonClassOverride:Common Button Base Class Reference=@n1
    self:WBP UIKit Side Navigation Base Object Reference=Get WBP_UIKit_SideNavigation()
   ) -> (then, Output_Get:Common Button Base Class Reference)
  ]}
 Set Side Nav CustomButton {nodes [
   Set Side Nav CustomButton() {is=title K2Node_FunctionEntry}
   n1 = Get SideNav_Button_TopLevelClass() -> (SideNav_Button_TopLevelClass:WBP UIKit Button Generic Base Class Reference)
   > Branch(Condition=IsValidClass(Class=@n1))
   > Set ButtonClass!(execute, ButtonClass:Common Button Base Class Reference=@n1, self:WBP UIKit Tab Group Base Object Reference=Get WBP_UIKit_SideNavigation()) -> (then, Output_Get:Common Button Base Class Reference)
  ]}
 SetAlternateContentNavigableWidget {nodes [
   n1 = SetAlternateContentNavigableWidget!() -> (then, NavigableWidget:Widget Object Reference, bIsAlternateContent:bool=~) {is=title K2Node_FunctionEntry}
   > Set AlternateContentNavigableWidget(AlternateContentNavigableWidget:Widget Object Reference=@n1) -> (Output_Get:Widget Object Reference)
   > Set bIsCurrentTabAlternateContent(bIsCurrentTabAlternateContent:bool=@n1.bIsAlternateContent) -> (Output_Get:bool)
  ]}
 SetCompactUIStyle {nodes [
   SetCompactUIStyle() {is=title K2Node_FunctionEntry}
   > Branch(Condition=IsMobileGame())
   n1 = Get SafeZone()
   > SetPadding!(
    execute
    self:Grid Slot Object Reference=SlotAsGridSlot(Widget=@n1)
    InPadding_Left:float=56
    InPadding_Top:float=164
    InPadding_Right:float=56
    InPadding_Bottom:float=0
   ) -> (then)
   > SetSidesToPad!(execute, self:Safe Zone Object Reference=@n1, InPadLeft:bool=true, InPadRight:bool=true, InPadTop:bool=true, InPadBottom:bool=true) -> (then)
  ]}
 SetLoadingState {nodes [
   n1 = SetLoadingState!() -> (then, LoadingState:EDiscoverLoadingState Enum=Unset) {is=title K2Node_FunctionEntry}
   > n2 = Switch on EDiscoverLoadingState!(execute, Selection:EDiscoverLoadingState Enum=@n1) -> (Unset:Exec, Loading:Exec, Loaded:Exec, Error:Exec) {is=title K2Node_SwitchEnum}
   SetActiveWidget(
    execute=@n2.Error
    self:Widget Switcher Object Reference=Get Switcher_LoadingState() -> (Switcher_LoadingState:Common Activatable Widget Switcher Object Reference)
    Widget=Get ErrorMessage() -> (ErrorMessage:WBP Discovery Blank Tiles Rounded Corners Object Reference)
   )
   SetActiveWidget(execute=@n2.Loaded, self=Get Switcher_LoadingState(), Widget=Get MainContent() -> (MainContent:Common Visibility Switcher Object Reference))
   SetActiveWidget(execute=[@n2.Loading, @n2.Unset], self=Get Switcher_LoadingState(), Widget=Get Loading_SpinnerContainer())
  ]}
 SetNoResultsSubtitleText {nodes [
   n1 = SetNoResultsSubtitleText!() -> (then, NoResultsSubText:Text) {is=title K2Node_FunctionEntry}
   > n2 = Set NoResultsSubText(NoResultsSubText:Text=@n1) -> (Output_Get:Text)
   > Set FlavorSubtitleOverride!(execute, FlavorSubtitleOverride:Text=@n2, self:WBP Discovery Blank Tiles Rounded Corners Object Reference=Get NoResultsGenericFallback()) -> (then, Output_Get:Text)
  ]}
 SetTemplateOverrides {nodes [
   SetTemplateOverrides() {is=title K2Node_FunctionEntry}
   > Set Side Nav CustomButton() {is=title K2Node_CallFunction}
   > Set Side Nav Custom Nested Button() {is=title K2Node_CallFunction}
   > Set Side Nav Custom Divider() {is=title K2Node_CallFunction}
   > Set Custom Loading Spinner() {is=title K2Node_CallFunction}
   > SetNoResultsSubtitleText!(execute, self, NoResultsSubText:Text=Get NoResultsSubText() -> (NoResultsSubText:Text)) -> (then) {is=title K2Node_CallFunction}
  ]}
 SetupSideNavAndMegaListFromViewModel {nodes [
   n1 = SetupSideNavAndMegaListFromViewModel!() -> (then, Tree View Model:Discover Panel Tree View Model Object Reference) {is=title K2Node_FunctionEntry}
   > ClearSideNavigationTabs(self:WBP UIKit Side Navigation Base Object Reference=Get WBP_UIKit_SideNavigation())
   > ClearListItems(self=Get ListView(self=Get WBP_Poly_MegaList()))
   > Clear(TargetArray:Array of MVVM Base Viewmodel Object References=Get MegaListToSideNavTabVMMapping())
   > n2 = Recursive Tree Creation(Node=@n1, IsIndented=false)
   > n3 = Branch(Condition=NOT Boolean(A=@n2))
   > ClearSideNavigationTabs(self=Get WBP_UIKit_SideNavigation())
   > n4 = ClearListItems(self=Get ListView(self=Get WBP_Poly_MegaList()))
   n5 = GetTabCount!(self:Common Tab List Widget Base Object Reference=Get WBP_UIKit_SideNavigation()) -> (ReturnValue:int=0)
   Call On Mega List Array Finished Recursive Creation!(execute:Exec=[@n4.then, @n3.else], self:WBP Template Nested Side Nav And Continuous List Object Reference, LastIndex:int=int - int(A=@n5)) -> (then:Exec) {is=title K2Node_CallDelegate}
   n6 = "integer > integer"(A=@n5)
   > SetActiveWidget(
    self:Common Visibility Switcher Object Reference=Get MainContent()
    Widget=Select(Option 0=Get EmptyList_NoResults(), Option 1=Get ListPlusTabs() -> (ListPlusTabs:Horizontal Box Object Reference), Index=@n6)
   )
   > Branch(Condition=@n6)
   > SelectTabByIndex!(execute, self:UIKit Tab Group Object Reference=Get WBP_UIKit_SideNavigation(), Index:int=0, bSuppressClickFeedback:bool=true) -> (then, ReturnValue:bool)
  ]}
 SwitchMainContent {nodes [
   SwitchMainContent() {is=title K2Node_FunctionEntry}
   > SetActiveWidget(
    self=Get Switcher_ContentType() -> (Switcher_ContentType:Common Visibility Switcher Object Reference)
    Widget=Select(
     Option 0=Get MegaListContainerOverlay()
     Option 1=Get Slot_AlternateContent() -> (Slot_AlternateContent:Named Slot Object Reference)
     Index=Get bIsCurrentTabAlternateContent()
    )
   )
  ]}
]
variables [
 AlternateContentNavigableWidget {type=Widget, default=null}
 bIsCurrentTabAlternateContent {type=boolean}
 IntroWhooshSound {type=SoundBase, default=/Game/Sounds/Fort_UI_Sounds/FrontEnd/Tabs/UI_Whoosh_TopBar_Tab_Home_Cue}
 LoadingSpinnerClass {type=Class, default=/FortUILibrary/UIKit/Throbber/WBP_UIKit_Throbber.WBP_UIKit_Throbber_C}
 LoadingSpinnerWidget {type=WBP_UIKit_Throbber_Base_C, default=null}
 MegaListPadding {type=Margin, default=<ltrb -38 0 80 0>}
 MegaListToSideNavTabVMMapping {type="array<MVVMViewModelBase>", default []}
 NoResultsSubText {type=string, default=Please try again later}
 OnMegaListArrayFinishedRecursiveCreation {}
 OnSideNavOuterTabClicked {}
 PaddingContentArea {type=Margin, default=<ltrb 0 164 0 0>}
 PreviousSelectedTileIndex {type=integer}
 SideNav_Button_NestedClass {type=Class, default=null}
 SideNav_Button_TopLevelClass {type=Class, default=/FortUILibrary/UIKit/Buttons/TabButtons/WBP_UIKit_SideNavigation_Button.WBP_UIKit_SideNavigation_Button_C}
 SideNav_DividerClass {type=Class, default=/FortUILibrary/Systems/Tabs/WBP_UIKit_Divider_Base.WBP_UIKit_Divider_Base_C}
 SideNav_TabContainerClass {type=Class, default=/FortUILibrary/Systems/Tabs/WBP_UIKit_TabContainer_Base.WBP_UIKit_TabContainer_Base_C}
]
widgets {
 info {
  inheritedWidgetCount=0
  parentClass=&CommonActivatableWidget
  rootWidgetClass=/Script/FortniteUI.FortSwipePanel
  widgetCount=19
 }
 widgets [
  SwipePanel {
   widgetClassPath=&FortSwipePanel
   bIsVariable=true
   widget=SwipePanel
   properties {bBeginSwipeOnPointerEnter=false, bConsumePointerInput=false, thresholdUnit=Inches, swipeThreshold=<0 0.25>}
  }
  RootGrid_SideNavWithList {
   widgetClassPath=&GridPanel
   parent=SwipePanel
   widget=RootGrid_SideNavWithList
   properties {columnFill [1], rowFill [1]}
   slot {class=FortSwipePanelSlot}
  }
  SafeZone {
   widgetClassPath=&SafeZone
   bIsVariable=true
   parent=RootGrid_SideNavWithList
   widget=SafeZone
   properties {padBottom=false, padLeft=false, padRight=false, padTop=false}
   slot {class=GridSlot, properties {layer=10, padding=<ltrb 80 164 0 0>}}
  }
  Switcher_LoadingState {
   widgetClassPath=&CommonActivatableWidgetSwitcher
   bIsVariable=true
   parent=SafeZone
   widget=Switcher_LoadingState
   slot {class=SafeZoneSlot}
  }
  MainContent {
   widgetClassPath=&CommonVisibilitySwitcher
   bIsVariable=true
   parent=Switcher_LoadingState
   widget=MainContent
   slot {class=WidgetSwitcherSlot}
  }
  ListPlusTabs {
   widgetClassPath=&HorizontalBox
   bIsVariable=true
   parent=MainContent
   widget=ListPlusTabs
   slot {class=CommonVisibilitySwitcherSlot}
  }
  VerticalBox_LeftSide {
   widgetClassPath=&VerticalBox
   parent=ListPlusTabs
   widget=VerticalBox_LeftSide
   slot {class=HorizontalBoxSlot, properties {verticalAlignment=VAlign_Top, padding=<ltrb 0 38 0 0>}}
  }
  WBP_UIKit_SideNavigation {
   widgetClassPath=/FortUILibrary/UIKit/Tabs/WBP_UIKit_SideNavigation
   bIsVariable=true
   parent=VerticalBox_LeftSide
   widget=WBP_UIKit_SideNavigation
   properties {
    hidePrompts=true
    navigation=&self:WidgetTree.WBP_UIKit_SideNavigation.WidgetNavigation_0
    selectInnerTabOnContainerExpand=true
   }
   slot {class=VerticalBoxSlot}
  }
  Slot_ContentBelowSideNav {
   widgetClassPath=&NamedSlot
   bIsVariable=true
   parent=VerticalBox_LeftSide
   widget=Slot_ContentBelowSideNav
   slot {class=VerticalBoxSlot, properties {padding=<ltrb 0 34 0 0>}}
  }
  Switcher_ContentType {
   widgetClassPath=&CommonVisibilitySwitcher
   bIsVariable=true
   parent=ListPlusTabs
   widget=Switcher_ContentType
   properties {navigation=&self:WidgetTree.Switcher_ContentType.WidgetNavigation_0}
   slot {class=HorizontalBoxSlot, properties {size {sizeRule=Fill, value=1}}}
  }
  MegaListContainerOverlay {
   widgetClassPath=&Overlay
   bIsVariable=true
   parent=Switcher_ContentType
   widget=MegaListContainerOverlay
   properties {clipping=ClipToBoundsWithoutIntersecting}
   slot {class=CommonVisibilitySwitcherSlot, properties {padding=<ltrb -24 0 0 0>}}
  }
  Image_HeaderSoftEdge {
   widgetClassPath=&Image
   bIsVariable=true
   parent=MegaListContainerOverlay
   widget=Image_HeaderSoftEdge
   properties {
    renderOpacity=0
    visibility=Collapsed
    brush {
     bIsDynamicallyLoaded=false
     drawAs=Image
     imageType=NoImage
     mirroring=NoMirror
     resourceName=None
     resourceObject=/DiscoveryBrowser/TabScreens/DiscoverHome/Materials/MI_UI_Discovery_NebulousBackgroundWithTopMask
     tiling=NoTile
     imageSize=<32 50>
     margin=<ltrb 0 0 0 0>
     outlineSettings {
      bUseBrushTransparency=false
      roundingType=HalfHeightRadius
      width=0
      color {colorUseRule=UseColor_Specified, specifiedColor=<rgba 0 0 0 0>}
      cornerRadii=<0 0 0 0>
     }
     tintColor {colorUseRule=UseColor_Specified, specifiedColor=<rgba 1 1 1 1>}
     uVRegion {bIsValid=false, max=<0 0>, min=<0 0>}
    }
    colorAndOpacity=<rgba 1 1 1 0>
   }
   slot {class=OverlaySlot, properties {horizontalAlignment=HAlign_Fill, padding=<ltrb 20 0 0 0>}}
  }
  WBP_Poly_MegaList {
   widgetClassPath=&/WBP_Poly_MegaList
   bIsVariable=true
   parent=MegaListContainerOverlay
   widget=WBP_Poly_MegaList
   slot {class=OverlaySlot, properties {horizontalAlignment=HAlign_Fill, verticalAlignment=VAlign_Fill}}
  }
  Slot_AlternateContent {
   widgetClassPath=&NamedSlot
   bIsVariable=true
   parent=Switcher_ContentType
   widget=Slot_AlternateContent
   slot {class=CommonVisibilitySwitcherSlot}
  }
  EmptyList_NoResults {
   widgetClassPath=&CommonVisibilitySwitcher
   bIsVariable=true
   parent=MainContent
   widget=EmptyList_NoResults
   slot {class=CommonVisibilitySwitcherSlot}
  }
  NoResultsGenericFallback {
   widgetClassPath=&/WBP_Discovery_BlankTiles_RoundedCorners
   bIsVariable=true
   parent=EmptyList_NoResults
   widget=NoResultsGenericFallback
   properties {entries Count=36, flavorSubtitleOverride=No content has been added to this page, flavorTitleOverride=Nothing Here Yet}
   slot {class=CommonVisibilitySwitcherSlot}
  }
  Slot_CustomNoResults {
   widgetClassPath=&NamedSlot
   bIsVariable=true
   parent=EmptyList_NoResults
   widget=Slot_CustomNoResults
   slot {class=CommonVisibilitySwitcherSlot}
  }
  ErrorMessage {
   widgetClassPath=&WBP_Discovery_BlankTiles_RoundedCorners
   bIsVariable=true
   parent=Switcher_LoadingState
   widget=ErrorMessage
   properties {entries Count=36, flavorSubtitleOverride=Refresh the page or try again later, flavorTitleOverride=Unable to Load Content}
   slot {class=WidgetSwitcherSlot, properties {padding=<ltrb 0 0 80 0>}}
  }
  Loading_SpinnerContainer {
   widgetClassPath=&Overlay
   bIsVariable=true
   parent=Switcher_LoadingState
   widget=Loading_SpinnerContainer
   slot {class=WidgetSwitcherSlot, properties {padding=<ltrb 0 0 80 0>}}
  }
 ]
}
